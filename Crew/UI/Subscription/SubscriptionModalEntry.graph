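# Bulleted line-item widget for the Crew subscription modal, showing a small circle pip beside an auto-wrapping rich text description.
# The SetText event writes incoming text to the Description block and stores it in the text variable.
# PreConstruct and Construct both reapply that variable, so the default sample text with inline markup previews in the designer and stored text is restored on construct.
# The row's layout direction follows the culture.
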WidgetBlueprint /Crew/UI/Subscription/SubscriptionModalEntry
graphs [EventGraph {nodes [
   n1 = SetText!() -> (OutputDelegate, then, InText:Text) {is=title K2Node_CustomEvent}
   > SetText(self=Get Description() -> (Description:Common Rich Text Block Object Reference), InText=@n1.InText) {is=title K2Node_CallFunction}
   > Set text(text:Text=@n1.InText) -> (Output_Get:Text)
   n2 = Event PreConstruct()
   SetText(execute=[@n2.then, @n3.then], self=Get Description(), InText=Get text() -> (text:Text))
   n3 = Event Construct()
  ]}]
variables [text {
  type=string
  default="This is the <text color=\"FFFFFFFF\" fontface=\"black\">description</>, this is the wrapping policy when the string is long."
 }]
widgets {
 info {inheritedWidgetCount=0, parentClass=&UserWidget, rootWidgetClass=&Overlay, widgetCount=4}
 widgets [
  Overlay_0 {widgetClassPath=&Overlay, widget=Overlay_0}
  HorizontalBox_14 {
   widgetClassPath=&HorizontalBox
   parent=Overlay_0
   widget=HorizontalBox_14
   properties {flowDirectionPreference=Culture}
   slot {class=OverlaySlot, properties {horizontalAlignment=HAlign_Fill, padding=<ltrb 0 4 0 16>}}
  }
  Pip {
   widgetClassPath=&Image
   bIsVariable=true
   parent=HorizontalBox_14
   widget=Pip
   properties {brush {
     bIsDynamicallyLoaded=false
     drawAs=Image
     imageType=FullColor
     mirroring=NoMirror
     resourceName=None
     resourceObject=/FortUILibrary/Foundation/Art_ToolKit/Tools/M_UI_Circle_White
     tiling=NoTile
     imageSize=<6 6>
     margin=<ltrb 0 0 0 0>
     outlineSettings {
      bUseBrushTransparency=false
      roundingType=HalfHeightRadius
      width=0
      color {colorUseRule=UseColor_Specified, specifiedColor=<rgba 0 0 0 0>}
      cornerRadii=<0 0 0 0>
     }
     tintColor {colorUseRule=UseColor_Specified, specifiedColor=<rgba 1 1 1 1>}
     uVRegion {bIsValid=false, max=<0 0>, min=<0 0>}
    }}
   slot {class=HorizontalBoxSlot, properties {horizontalAlignment=HAlign_Left, verticalAlignment=VAlign_Center, padding=<ltrb 0 0 16 0>}}
  }
  Description {
   widgetClassPath=&CommonRichTextBlock
   bIsVariable=true
   parent=HorizontalBox_14
   widget=Description
   properties {
    autoWrapText=true
    bOverrideDefaultStyle=true
    lineHeightPercentage=0.875
    mobileTextBlockScale=1.2
    text="This is the <text color=\"FFFFFFFF\" fontface=\"black\">description</>, this is the wrapping policy when the string is long."
    defaultTextStyleOverride {
     overflowPolicy=Clip
     transformPolicy=None
     colorAndOpacity {colorUseRule=UseColor_Specified, specifiedColor=<rgba 0.5776 0.5776 0.5776 1>}
     font {
      bForceMonospaced=false
      bMaterialIsStencil=false
      fontMaterial=null
      fontObject=&HeadingNow
      letterSpacing=0
      monospacedWidth=1
      size=15.75
      skewAmount=0
      typefaceFontName=74 Regular
      outlineSettings {
       bApplyOutlineToDropShadows=false
       bMiteredCorners=false
       bSeparateFillAlpha=false
       outlineMaterial=null
       outlineSize=0
       outlineColor=<rgba 0 0 0 1>
      }
     }
     highlightColor {colorUseRule=UseColor_Specified, specifiedColor=<rgba 0 0 0 1>}
     highlightShape {
      bIsDynamicallyLoaded=false
      drawAs=Image
      imageType=NoImage
      mirroring=NoMirror
      resourceName=None
      resourceObject=null
      tiling=NoTile
      imageSize=<32 32>
      margin=<ltrb 0 0 0 0>
      outlineSettings {
       bUseBrushTransparency=false
       roundingType=HalfHeightRadius
       width=0
       color {colorUseRule=UseColor_Specified, specifiedColor=<rgba 0 0 0 0>}
       cornerRadii=<0 0 0 0>
      }
      tintColor {colorUseRule=UseColor_Specified, specifiedColor=<rgba 1 1 1 1>}
      uVRegion {bIsValid=false, max=<0 0>, min=<0 0>}
     }
     selectedBackgroundColor {colorUseRule=UseColor_Foreground, specifiedColor=<rgba 1 0 1 1>}
     shadowColorAndOpacity=<rgba 0 0 0 1>
     shadowOffset=<0 0>
     strikeBrush {
      bIsDynamicallyLoaded=false
      drawAs=Image
      imageType=NoImage
      mirroring=NoMirror
      resourceName=None
      resourceObject=null
      tiling=NoTile
      imageSize=<32 32>
      margin=<ltrb 0 0 0 0>
      outlineSettings {
       bUseBrushTransparency=false
       roundingType=HalfHeightRadius
       width=0
       color {colorUseRule=UseColor_Specified, specifiedColor=<rgba 0 0 0 0>}
       cornerRadii=<0 0 0 0>
      }
      tintColor {colorUseRule=UseColor_Specified, specifiedColor=<rgba 1 1 1 1>}
      uVRegion {bIsValid=false, max=<0 0>, min=<0 0>}
     }
     underlineBrush {
      bIsDynamicallyLoaded=false
      drawAs=Image
      imageType=NoImage
      mirroring=NoMirror
      resourceName=None
      resourceObject=null
      tiling=NoTile
      imageSize=<32 32>
      margin=<ltrb 0 0 0 0>
      outlineSettings {
       bUseBrushTransparency=false
       roundingType=HalfHeightRadius
       width=0
       color {colorUseRule=UseColor_Specified, specifiedColor=<rgba 0 0 0 0>}
       cornerRadii=<0 0 0 0>
      }
      tintColor {colorUseRule=UseColor_Specified, specifiedColor=<rgba 1 1 1 1>}
      uVRegion {bIsValid=false, max=<0 0>, min=<0 0>}
     }
    }
   }
   slot {class=HorizontalBoxSlot, properties {verticalAlignment=VAlign_Center, size {sizeRule=Fill, value=1}}}
  }
 ]
}
