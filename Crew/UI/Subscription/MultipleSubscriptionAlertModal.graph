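# Activatable modal warning players that they hold Fortnite Crew subscriptions on multiple platforms and risk multiple billing.
# PreConstruct feeds a DialogVM_MultipleSubscription into the UIKit dialog base.
# OnSubscriptionPlatformsUpdated rebuilds the bullet list, using SupportShowOtherPlatforms to choose between named platforms with a cancellation link or generic text with a check-subscriptions link.
# Dialog buttons and touch close save the do-not-remind toggle to FortLocalPlayerVM, then deactivate the widget.
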WidgetBlueprint /Crew/UI/Subscription/MultipleSubscriptionAlertModal
graphs [
 EventGraph {nodes [
   "On Clicked (Button_TextURL_1)"!() -> (OutputDelegate, then, Button:Common Button Base Object Reference) {is=title K2Node_ComponentBoundEvent}
   > LaunchURL!(execute, URL:String=Get CheckCrewURL() -> (CheckCrewURL:String)) -> (then)
   n1 = "On Clicked (Button_TouchClose)"!() -> (OutputDelegate, then, Button:Common Button Base Object Reference) {is=title K2Node_ComponentBoundEvent}
   SetDoNotRemindAboutMultiSubs!(
    execute=[@n1.then, @n3.then]
    self:Fort Local Player VM Object Reference=Get FortLocalPlayerVM() -> (FortLocalPlayerVM:Fort Local Player VM Object Reference)
    bDoNotRemind:bool=Get CloseWithDoNotRemind() -> (CloseWithDoNotRemind:bool)
   ) -> (then)
   > DeactivateWidget()
   "On Clicked (Button_TextURL)"!() -> (OutputDelegate, then, Button:Common Button Base Object Reference) {is=title K2Node_ComponentBoundEvent}
   > LaunchURL(URL=Get HowToCancelURL() -> (HowToCancelURL:String))
   n2 = "On Selected Changed (WBP_UIKit_ButtonWithToggle)"!() -> (OutputDelegate, then, Button:Common Button Base Object Reference, Selected:bool) {is=title K2Node_ComponentBoundEvent}
   > Set CloseWithDoNotRemind(CloseWithDoNotRemind:bool=@n2.Selected) -> (Output_Get:bool)
   n3 = CloseClicked!() -> (OutputDelegate, then, ButtonIndex:int) {is=title K2Node_CustomEvent}
   Event PreConstruct()
   > n4 = Construct Object from Class(Class=/Crew/UI/Subscription/Data/DialogVM_MultipleSubscription.DialogVM_MultipleSubscription_C) -> (ReturnValue:Dialog VM Multiple Subscription Object Reference)
   > SetDialogData!(
    execute
    self:WBP UIKit Dialog Internal Object Reference=Get WBP_UIKit_Dialog_Base() -> (WBP_UIKit_Dialog_Base:WBP UIKit Dialog Internal Object Reference)
    Data:UI Kit Dialog Viewmodel Object Reference=@n4
   ) -> (then)
   > SetScrollWidget!(execute, ToScroll:Widget Object Reference=Get PrimaryScrollBox() -> (PrimaryScrollBox:Scroll Box Object Reference)) -> (then)
   Event On Activated()
   n5 = Create Event()
   n6 = Get WBP_UIKit_Dialog_Base()
   > Bind Event to Dialog Button Clicked Dispatcher(self:WBP UIKit Dialog Internal Object Reference=@n6, Delegate=@n5) {is=title K2Node_AddDelegate}
   Event On Deactivated()
   > Unbind Event from Dialog Button Clicked Dispatcher(self:WBP UIKit Dialog Internal Object Reference=@n6, Delegate=@n5) {is=title K2Node_RemoveDelegate}
  ]}
 GetCanLaunchURLVisibility {nodes [
   n1 = GetCanLaunchURLVisibility!() -> (then, URL:String) {is=title K2Node_FunctionEntry}
   > n2 = CanLaunchURL!(execute, URL:String=@n1) -> (then, ReturnValue:bool)
   > Return Node!(execute, Result:ESlateVisibility Enum=Select(Option 0=Collapsed, Option 1=Visible, Index=@n2))
  ]}
 OnSubscriptionPlatformsUpdated {nodes [
   OnSubscriptionPlatformsUpdated!() -> (then, ActiveSubscriptions:Array of EAppStore Enums=DebugStore) {is=title K2Node_FunctionEntry}
   > Reset!(
    execute
    self:Dynamic Entry Box Object Reference=Get EntryBox_BulletPoints() -> (EntryBox_BulletPoints:Dynamic Entry Box Object Reference)
    bDeleteWidgets:bool=true
   ) -> (then)
   > n1 = Sequence()
   > n2 = SupportShowOtherPlatforms!(execute) -> (then, ReturnValue:bool)
   > SetActiveWidget(
    self=Get LinkSwitcher() -> (LinkSwitcher:Common Visibility Switcher Object Reference)
    Widget=Select(
     Option 0:Widget Object Reference=Get NonEGSLink() -> (NonEGSLink:Horizontal Box Object Reference)
     Option 1:Widget Object Reference=Get EGSLink() -> (EGSLink:Horizontal Box Object Reference)
     Index=@n2
    ) -> (ReturnValue:Widget Object Reference)
   )
   > n3 = Branch(Condition=@n2)
   > n4 = For Each Loop(Array:Array of EAppStore Enums=Get ActiveSubscriptions() -> (ActiveSubscriptions:Array of EAppStore Enums=DebugStore)) -> (Array Element:EAppStore Enum)
   n5 = Create Entry of Class!(
    execute=@n4.Completed
    self:Dynamic Entry Box Object Reference=Get EntryBox_BulletPoints()
    EntryClass:User Widget Class Reference=&/SubscriptionModalEntry.SubscriptionModalEntry_C
   ) -> (then, ReturnValue:Subscription Modal Entry Object Reference)
   > SetText(
    self:Subscription Modal Entry Object Reference=@n5
    InText:Text=Format Text!(
     Format=Get PlatformBulletFormat_SupportOtherPlatforms() -> (PlatformBulletFormat_SupportOtherPlatforms:Text)
     Platforms:Text=Get PlatformBulletText() -> (PlatformBulletText:Text)
    ) -> (Result)
   )
   n6 = Find(
    TargetMap:Map of EAppStore Enums to Texts=Get AppStorePlatformMapping() -> (AppStorePlatformMapping:Map of EAppStore Enums to Texts=DebugStore)
    Key:"EAppStore Enum (by ref)"=@n4
   ) -> (Value:Text)
   Branch(execute=@n4.LoopBody, Condition=@n6.ReturnValue)
   n7 = TextIsEmpty(InText=Get PlatformBulletText())
   n8 = Get PlatformDelimiter() -> (PlatformDelimiter:Text)
   > Set PlatformBulletText(PlatformBulletText:Text=Format Text!(
     Format="LOCTEXT(\"{existing}{delim}{platformtext}\")"
     existing:Text=Get PlatformBulletText()
     delim:Text=Select(Option 0:Text=@n8, Option 1:Text, Index=@n7) -> (ReturnValue:Text)
     platformtext:Text=@n6
    ) -> (Result)) -> (Output_Get:Text)
   n9 = Create Entry of Class(execute=@n3.else, self=Get EntryBox_BulletPoints(), EntryClass=&SubscriptionModalEntry_C)
   > SetText(self=@n9, InText=Get PlatformBulletFormat_Generic() -> (PlatformBulletFormat_Generic:Text))
   n10 = For Each Loop(Exec=@n1.then_1, Array:Array of Texts=Get BulletPoints() -> (BulletPoints:Array of Texts)) -> (Array Element:Text)
   > n11 = Create Entry of Class(self=Get EntryBox_BulletPoints(), EntryClass=&SubscriptionModalEntry_C)
   > SetText(self=@n11, InText=@n10)
  ]}
]
variables [
 AppStorePlatformMapping {default {
   epicPurchasingService=Epic Games Store
   iOSAppStore=App Store
   microsoftStore=Microsoft Store
   nintendoEShop=Nintendo eShop
   oneStore=ONE store
   playStation4Store="PlayStation™Store"
   samsungGalaxyAppStore=Samsung Galaxy Store
  }}
 BulletPoints {
  type="array<string>"
  default [
   "In order to avoid being <text color=\"FFFFFFFF\" fontface=\"black\">BILLED MULTIPLE TIMES</>, we recommend cancelling the extra subscriptions."
   "If you remain subscribed on multiple platforms, you will not receive multiple copies of the Fortnite Crew Rewards. Instead, you will receive 1,500 V-Bucks for every extra subscription every billing period."
  ]
 }
 CheckCrewURL {type=string, default=https://fn.gg/check-crew}
 CloseWithDoNotRemind {type=boolean}
 HowToCancelURL {type=string, default=https://fn.gg/cancel-crew}
 PlatformBulletFormat_Generic {type=string, default=You are currently subscribed to the Fortnite Crew on multiple platforms.}
 PlatformBulletFormat_SupportOtherPlatforms {type=string, default="You are currently subscribed to the Fortnite Crew on {Platforms}."}
 PlatformDelimiter {type=string, default=", "}
 PlatformTextStyle {type=string, default="<text color=\"FFFFFFFF\">{Platform}</>"}
]
widgets {
 info {
  inheritedWidgetCount=0
  namedSlotCount=1
  parentClass=&CommonActivatableWidget
  rootWidgetClass=&Overlay
  widgetCount=20
 }
 widgets [
  Overlay_Root {widgetClassPath=&Overlay, widget=Overlay_Root}
  WBP_UIKit_Dialog_Base {
   widgetClassPath=/FortUILibrary/Systems/Dialog/WBP_UIKit_Dialog_Internal
   bIsVariable=true
   parent=Overlay_Root
   widget=WBP_UIKit_Dialog_Base
   properties {
    useText=None
    buttons {
     each Button Horizontal Fill=false
     each Button Min Width=0
     each Button Size=Medium
     focuse Button=0
     space Between Buttons=30
     cTA Buttons Types [Regular]
     group Placement {
      horizontal Alignment=HAlign_Left
      vertical Alignment=VAlign_Bottom
      grid {column=1, column Span=0, layer=0, row=2, row Span=0, nudge=<0 0>}
      padding=<ltrb 80 64 80 48>
     }
    }
    dialogSIze {
     height Override {enable=false, value=700}
     max Aspect Ratio {enable=false, value=0}
     max Desired Height {enable=true, value=800}
     max Desired Width {enable=false, value=0}
     min Aspect Ratio {enable=false, value=0}
     min Desired Height {enable=false, value=0}
     min Desired Width {enable=false, value=0}
     width Override {enable=true, value=1490}
    }
   }
   slot {class=OverlaySlot, properties {horizontalAlignment=HAlign_Fill, verticalAlignment=VAlign_Fill}}
  }
  Overlay_111 {widgetClassPath=&Overlay, namedSlotHost=WBP_UIKit_Dialog_Base, widget=Overlay_111}
  TextContent {
   widgetClassPath=&VerticalBox
   parent=Overlay_111
   widget=TextContent
   slot {class=OverlaySlot, properties {horizontalAlignment=HAlign_Fill, verticalAlignment=VAlign_Fill, padding=<ltrb 80 0 80 0>}}
  }
  PrimaryScrollBox {
   widgetClassPath=&ScrollBox
   bIsVariable=true
   parent=TextContent
   widget=PrimaryScrollBox
   properties {
    widgetBarStyle {
     thickness=16
     draggedThumbImage {
      bIsDynamicallyLoaded=false
      drawAs=Image
      imageType=FullColor
      mirroring=NoMirror
      resourceName=../../../Engine/Content/Slate/Common/Scrollbar_Thumb.png
      resourceObject=/Game/Athena/UI/ReplayBrowser/M_Ui_Skewed_Vfor9slices_BG_Separator_Inst
      tiling=NoTile
      imageSize=<0 0>
      margin=<ltrb 0.25 0.25 0.25 0.25>
      outlineSettings {
       bUseBrushTransparency=false
       roundingType=HalfHeightRadius
       width=0
       color {colorUseRule=UseColor_Specified, specifiedColor=<rgba 0 0 0 0>}
       cornerRadii=<0 0 0 0>
      }
      tintColor {colorUseRule=UseColor_Specified, specifiedColor=<rgba 1 1 1 1>}
      uVRegion {bIsValid=false, max=<0 0>, min=<0 0>}
     }
     horizontalBackgroundImage {
      bIsDynamicallyLoaded=false
      drawAs=Image
      imageType=NoImage
      mirroring=NoMirror
      resourceName=None
      resourceObject=&M_Ui_Skewed_Vfor9slices_BG_Separator_Inst
      tiling=NoTile
      imageSize=<0 0>
      margin=<ltrb 0 0 0 0>
      outlineSettings {
       bUseBrushTransparency=false
       roundingType=HalfHeightRadius
       width=0
       color {colorUseRule=UseColor_Specified, specifiedColor=<rgba 0 0 0 0>}
       cornerRadii=<0 0 0 0>
      }
      tintColor {colorUseRule=UseColor_Specified, specifiedColor=<rgba 0 0.2159 1 0.498>}
      uVRegion {bIsValid=false, max=<0 0>, min=<0 0>}
     }
     horizontalBottomSlotImage {
      bIsDynamicallyLoaded=false
      drawAs=NoDrawType
      imageType=FullColor
      mirroring=NoMirror
      resourceName=../../../Engine/Content/Slate/Common/Scrollbar_Background_Horizontal.png
      resourceObject=null
      tiling=NoTile
      imageSize=<8 8>
      margin=<ltrb 0 0 0 0>
      outlineSettings {
       bUseBrushTransparency=false
       roundingType=HalfHeightRadius
       width=0
       color {colorUseRule=UseColor_Specified, specifiedColor=<rgba 0 0 0 0>}
       cornerRadii=<0 0 0 0>
      }
      tintColor {colorUseRule=UseColor_Specified, specifiedColor=<rgba 1 1 1 1>}
      uVRegion {bIsValid=false, max=<0 0>, min=<0 0>}
     }
     horizontalTopSlotImage {
      bIsDynamicallyLoaded=false
      drawAs=NoDrawType
      imageType=FullColor
      mirroring=NoMirror
      resourceName=../../../Engine/Content/Slate/Common/Scrollbar_Background_Horizontal.png
      resourceObject=null
      tiling=NoTile
      imageSize=<8 8>
      margin=<ltrb 0 0 0 0>
      outlineSettings {
       bUseBrushTransparency=false
       roundingType=HalfHeightRadius
       width=0
       color {colorUseRule=UseColor_Specified, specifiedColor=<rgba 0 0 0 0>}
       cornerRadii=<0 0 0 0>
      }
      tintColor {colorUseRule=UseColor_Specified, specifiedColor=<rgba 1 1 1 1>}
      uVRegion {bIsValid=false, max=<0 0>, min=<0 0>}
     }
     hoveredThumbImage {
      bIsDynamicallyLoaded=false
      drawAs=Image
      imageType=FullColor
      mirroring=NoMirror
      resourceName=../../../Engine/Content/Slate/Common/Scrollbar_Thumb.png
      resourceObject=&M_Ui_Skewed_Vfor9slices_BG_Separator_Inst
      tiling=NoTile
      imageSize=<0 0>
      margin=<ltrb 0.25 0.25 0.25 0.25>
      outlineSettings {
       bUseBrushTransparency=false
       roundingType=HalfHeightRadius
       width=0
       color {colorUseRule=UseColor_Specified, specifiedColor=<rgba 0 0 0 0>}
       cornerRadii=<0 0 0 0>
      }
      tintColor {colorUseRule=UseColor_Specified, specifiedColor=<rgba 0.5573 0.808 1 1>}
      uVRegion {bIsValid=false, max=<0 0>, min=<0 0>}
     }
     normalThumbImage {
      bIsDynamicallyLoaded=false
      drawAs=Image
      imageType=FullColor
      mirroring=NoMirror
      resourceName=../../../Engine/Content/Slate/Common/Scrollbar_Thumb.png
      resourceObject=&M_Ui_Skewed_Vfor9slices_BG_Separator_Inst
      tiling=NoTile
      imageSize=<12 12>
      margin=<ltrb 0.25 0.25 0.25 0.25>
      outlineSettings {
       bUseBrushTransparency=false
       roundingType=HalfHeightRadius
       width=0
       color {colorUseRule=UseColor_Specified, specifiedColor=<rgba 0 0 0 0>}
       cornerRadii=<0 0 0 0>
      }
      tintColor {colorUseRule=UseColor_Specified, specifiedColor=<rgba 0.2122 0.6584 1 1>}
      uVRegion {bIsValid=false, max=<0 0>, min=<0 0>}
     }
     verticalBackgroundImage {
      bIsDynamicallyLoaded=false
      drawAs=Image
      imageType=NoImage
      mirroring=NoMirror
      resourceName=None
      resourceObject=&M_Ui_Skewed_Vfor9slices_BG_Separator_Inst
      tiling=NoTile
      imageSize=<0 0>
      margin=<ltrb 0 0 0 0>
      outlineSettings {
       bUseBrushTransparency=false
       roundingType=HalfHeightRadius
       width=0
       color {colorUseRule=UseColor_Specified, specifiedColor=<rgba 0 0 0 0>}
       cornerRadii=<0 0 0 0>
      }
      tintColor {colorUseRule=UseColor_Specified, specifiedColor=<rgba 0 0.2159 1 0.498>}
      uVRegion {bIsValid=false, max=<0 0>, min=<0 0>}
     }
     verticalBottomSlotImage {
      bIsDynamicallyLoaded=false
      drawAs=NoDrawType
      imageType=FullColor
      mirroring=NoMirror
      resourceName=../../../Engine/Content/Slate/Common/Scrollbar_Background_Horizontal.png
      resourceObject=null
      tiling=NoTile
      imageSize=<8 8>
      margin=<ltrb 0 0 0 0>
      outlineSettings {
       bUseBrushTransparency=false
       roundingType=HalfHeightRadius
       width=0
       color {colorUseRule=UseColor_Specified, specifiedColor=<rgba 0 0 0 0>}
       cornerRadii=<0 0 0 0>
      }
      tintColor {colorUseRule=UseColor_Specified, specifiedColor=<rgba 1 1 1 1>}
      uVRegion {bIsValid=false, max=<0 0>, min=<0 0>}
     }
     verticalTopSlotImage {
      bIsDynamicallyLoaded=false
      drawAs=NoDrawType
      imageType=FullColor
      mirroring=NoMirror
      resourceName=../../../Engine/Content/Slate/Common/Scrollbar_Background_Horizontal.png
      resourceObject=null
      tiling=NoTile
      imageSize=<8 8>
      margin=<ltrb 0 0 0 0>
      outlineSettings {
       bUseBrushTransparency=false
       roundingType=HalfHeightRadius
       width=0
       color {colorUseRule=UseColor_Specified, specifiedColor=<rgba 0 0 0 0>}
       cornerRadii=<0 0 0 0>
      }
      tintColor {colorUseRule=UseColor_Specified, specifiedColor=<rgba 1 1 1 1>}
      uVRegion {bIsValid=false, max=<0 0>, min=<0 0>}
     }
    }
    widgetStyle {
     barThickness=9
     bottomShadowBrush {
      bIsDynamicallyLoaded=false
      drawAs=NoDrawType
      imageType=FullColor
      mirroring=NoMirror
      resourceName=../../../Engine/Content/Slate/Common/ScrollBoxShadowLeft.png
      resourceObject=null
      tiling=NoTile
      imageSize=<8 16>
      margin=<ltrb 1 0.5 0 0.5>
      outlineSettings {
       bUseBrushTransparency=false
       roundingType=HalfHeightRadius
       width=0
       color {colorUseRule=UseColor_Specified, specifiedColor=<rgba 0 0 0 0>}
       cornerRadii=<0 0 0 0>
      }
      tintColor {colorUseRule=UseColor_Specified, specifiedColor=<rgba 1 1 1 1>}
      uVRegion {bIsValid=false, max=<0 0>, min=<0 0>}
     }
     horizontalScrolledContentPadding=<ltrb 0 0 1 0>
     leftShadowBrush {
      bIsDynamicallyLoaded=false
      drawAs=NoDrawType
      imageType=FullColor
      mirroring=NoMirror
      resourceName=../../../Engine/Content/Slate/Common/ScrollBoxShadowLeft.png
      resourceObject=null
      tiling=NoTile
      imageSize=<8 16>
      margin=<ltrb 1 0.5 0 0.5>
      outlineSettings {
       bUseBrushTransparency=false
       roundingType=HalfHeightRadius
       width=0
       color {colorUseRule=UseColor_Specified, specifiedColor=<rgba 0 0 0 0>}
       cornerRadii=<0 0 0 0>
      }
      tintColor {colorUseRule=UseColor_Specified, specifiedColor=<rgba 1 1 1 1>}
      uVRegion {bIsValid=false, max=<0 0>, min=<0 0>}
     }
     rightShadowBrush {
      bIsDynamicallyLoaded=false
      drawAs=NoDrawType
      imageType=FullColor
      mirroring=NoMirror
      resourceName=../../../Engine/Content/Slate/Common/ScrollBoxShadowRight.png
      resourceObject=null
      tiling=NoTile
      imageSize=<8 16>
      margin=<ltrb 0 0.5 1 0.5>
      outlineSettings {
       bUseBrushTransparency=false
       roundingType=HalfHeightRadius
       width=0
       color {colorUseRule=UseColor_Specified, specifiedColor=<rgba 0 0 0 0>}
       cornerRadii=<0 0 0 0>
      }
      tintColor {colorUseRule=UseColor_Specified, specifiedColor=<rgba 1 1 1 1>}
      uVRegion {bIsValid=false, max=<0 0>, min=<0 0>}
     }
     topShadowBrush {
      bIsDynamicallyLoaded=false
      drawAs=NoDrawType
      imageType=FullColor
      mirroring=NoMirror
      resourceName=../../../Engine/Content/Slate/Common/ScrollBoxShadowLeft.png
      resourceObject=null
      tiling=NoTile
      imageSize=<8 16>
      margin=<ltrb 1 0.5 0 0.5>
      outlineSettings {
       bUseBrushTransparency=false
       roundingType=HalfHeightRadius
       width=0
       color {colorUseRule=UseColor_Specified, specifiedColor=<rgba 0 0 0 0>}
       cornerRadii=<0 0 0 0>
      }
      tintColor {colorUseRule=UseColor_Specified, specifiedColor=<rgba 1 1 1 1>}
      uVRegion {bIsValid=false, max=<0 0>, min=<0 0>}
     }
     verticalScrolledContentPadding=<ltrb 0 0 0 1>
    }
   }
   slot {class=VerticalBoxSlot, properties {size {sizeRule=Fill, value=1}}}
  }
  EntryBox_BulletPoints {
   widgetClassPath=&DynamicEntryBox
   bIsVariable=true
   parent=PrimaryScrollBox
   widget=EntryBox_BulletPoints
   properties {entryBoxType=Vertical, entryWidgetClass=&SubscriptionModalEntry_C}
   slot {class=ScrollBoxSlot}
  }
  Acknowledgement {
   widgetClassPath=&HorizontalBox
   bIsVariable=true
   parent=TextContent
   widget=Acknowledgement
   slot {class=VerticalBoxSlot, properties {padding=<ltrb 0 36 0 0>}}
  }
  WBP_UIKit_ButtonWithToggle {
   widgetClassPath=/FortUILibrary/UIKit/Buttons/WBP_UIKit_ButtonWithToggle
   bIsVariable=true
   parent=Acknowledgement
   widget=WBP_UIKit_ButtonWithToggle
   properties {
    text=Do not remind again
    toggleType=CheckBox
    font {
     bForceMonospaced=false
     bMaterialIsStencil=false
     fontMaterial=/FortUILibrary/UIKit/Buttons/MaterialInstances/GenericButtons/MI_UIKit_ButtonWithToggle_Text
     fontObject=&HeadingNow
     letterSpacing=48
     monospacedWidth=1
     size=18.75
     skewAmount=0
     typefaceFontName=76 Bold
     outlineSettings {
      bApplyOutlineToDropShadows=false
      bMiteredCorners=false
      bSeparateFillAlpha=false
      outlineMaterial=null
      outlineSize=0
      outlineColor=<rgba 0 0 0 1>
     }
    }
    textPadding=<ltrb 14 8 0 0>
    triggeringInputAction {dataTable=&InputActionDataTable, rowName=GenericFaceButtonTop}
   }
   slot {class=HorizontalBoxSlot, properties {verticalAlignment=VAlign_Center}}
  }
  LinkSwitcher {
   widgetClassPath=&CommonVisibilitySwitcher
   bIsVariable=true
   parent=TextContent
   widget=LinkSwitcher
   slot {class=VerticalBoxSlot, properties {verticalAlignment=VAlign_Top, padding=<ltrb 0 20 0 0>}}
  }
  EGSLink {widgetClassPath=&HorizontalBox, bIsVariable=true, parent=LinkSwitcher, widget=EGSLink, slot {class=CommonVisibilitySwitcherSlot}}
  Text_MoreInfo {
   widgetClassPath=&CommonRichTextBlock
   bIsVariable=true
   parent=EGSLink
   widget=Text_MoreInfo
   properties {
    bOverrideDefaultStyle=true
    lineHeightPercentage=0.875
    mobileTextBlockScale=1.2
    text="To view Instructions on how to cancel a subscription on all platforms, please visit: "
    defaultTextStyleOverride {
     overflowPolicy=Clip
     transformPolicy=None
     colorAndOpacity {colorUseRule=UseColor_Specified, specifiedColor=<rgba 0.5776 0.5776 0.5776 1>}
     font {
      bForceMonospaced=false
      bMaterialIsStencil=false
      fontMaterial=null
      fontObject=&HeadingNow
      letterSpacing=0
      monospacedWidth=1
      size=15
      skewAmount=0
      typefaceFontName=74 Regular
      outlineSettings {
       bApplyOutlineToDropShadows=false
       bMiteredCorners=false
       bSeparateFillAlpha=false
       outlineMaterial=null
       outlineSize=0
       outlineColor=<rgba 0 0 0 1>
      }
     }
     highlightColor {colorUseRule=UseColor_Specified, specifiedColor=<rgba 0 0 0 1>}
     highlightShape {
      bIsDynamicallyLoaded=false
      drawAs=Image
      imageType=NoImage
      mirroring=NoMirror
      resourceName=None
      resourceObject=null
      tiling=NoTile
      imageSize=<32 32>
      margin=<ltrb 0 0 0 0>
      outlineSettings {
       bUseBrushTransparency=false
       roundingType=HalfHeightRadius
       width=0
       color {colorUseRule=UseColor_Specified, specifiedColor=<rgba 0 0 0 0>}
       cornerRadii=<0 0 0 0>
      }
      tintColor {colorUseRule=UseColor_Specified, specifiedColor=<rgba 1 1 1 1>}
      uVRegion {bIsValid=false, max=<0 0>, min=<0 0>}
     }
     selectedBackgroundColor {colorUseRule=UseColor_Foreground, specifiedColor=<rgba 1 0 1 1>}
     shadowColorAndOpacity=<rgba 0 0 0 1>
     shadowOffset=<0 0>
     strikeBrush {
      bIsDynamicallyLoaded=false
      drawAs=Image
      imageType=NoImage
      mirroring=NoMirror
      resourceName=None
      resourceObject=null
      tiling=NoTile
      imageSize=<32 32>
      margin=<ltrb 0 0 0 0>
      outlineSettings {
       bUseBrushTransparency=false
       roundingType=HalfHeightRadius
       width=0
       color {colorUseRule=UseColor_Specified, specifiedColor=<rgba 0 0 0 0>}
       cornerRadii=<0 0 0 0>
      }
      tintColor {colorUseRule=UseColor_Specified, specifiedColor=<rgba 1 1 1 1>}
      uVRegion {bIsValid=false, max=<0 0>, min=<0 0>}
     }
     underlineBrush {
      bIsDynamicallyLoaded=false
      drawAs=Image
      imageType=NoImage
      mirroring=NoMirror
      resourceName=None
      resourceObject=null
      tiling=NoTile
      imageSize=<32 32>
      margin=<ltrb 0 0 0 0>
      outlineSettings {
       bUseBrushTransparency=false
       roundingType=HalfHeightRadius
       width=0
       color {colorUseRule=UseColor_Specified, specifiedColor=<rgba 0 0 0 0>}
       cornerRadii=<0 0 0 0>
      }
      tintColor {colorUseRule=UseColor_Specified, specifiedColor=<rgba 1 1 1 1>}
      uVRegion {bIsValid=false, max=<0 0>, min=<0 0>}
     }
    }
   }
   slot {class=HorizontalBoxSlot, properties {horizontalAlignment=HAlign_Left, verticalAlignment=VAlign_Center, padding=<ltrb 0 0 0 2>}}
  }
  Button_OpenHowToURL {
   widgetClassPath=/Game/UI/SocialPanel/UserActionMenuInputButton
   bIsVariable=true
   parent=EGSLink
   widget=Button_OpenHowToURL
   properties {
    onButtonBaseClicked=(&self:WidgetTree.Button_OpenHowToURL.HandleOnButtonClicked)
    onButtonBaseDoubleClicked=(&self:WidgetTree.Button_OpenHowToURL.HandleOnButtonDoubleClicked)
    onButtonBaseHovered=(&self:WidgetTree.Button_OpenHowToURL.HandleOnButtonHovered)
    onButtonBaseUnhovered=(&self:WidgetTree.Button_OpenHowToURL.HandleOnButtonUnhovered)
    onSelectedChangedBase=(&self:WidgetTree.Button_OpenHowToURL.HandleOnSelectedChanged)
    triggeringInputAction {dataTable=&InputActionDataTable, rowName=PriceAcknowledge_MoreInfo}
   }
   slot {class=HorizontalBoxSlot, properties {horizontalAlignment=HAlign_Left, verticalAlignment=VAlign_Center}}
  }
  Button_TextURL {
   widgetClassPath=/Crew/UI/RichTextButton
   bIsVariable=true
   parent=EGSLink
   widget=Button_TextURL
   properties {
    auto wrap text=false
    button text="[PH]https://fn.gg/cancel-crew"
    fontSize=16
    fontSizeOverride=true
    onButtonBaseClicked=(&self:WidgetTree.Button_TextURL.HandleOnButtonClicked)
    onButtonBaseDoubleClicked=(&self:WidgetTree.Button_TextURL.HandleOnButtonDoubleClicked)
    onButtonBaseHovered=(&self:WidgetTree.Button_TextURL.HandleOnButtonHovered)
    onButtonBaseUnhovered=(&self:WidgetTree.Button_TextURL.HandleOnButtonUnhovered)
    onSelectedChangedBase=(&self:WidgetTree.Button_TextURL.HandleOnSelectedChanged)
   }
   slot {class=HorizontalBoxSlot, properties {horizontalAlignment=HAlign_Left, verticalAlignment=VAlign_Center}}
  }
  NonEGSLink {
   widgetClassPath=&HorizontalBox
   bIsVariable=true
   parent=LinkSwitcher
   widget=NonEGSLink
   slot {class=CommonVisibilitySwitcherSlot}
  }
  Text_MoreInfo_1 {
   widgetClassPath=&CommonRichTextBlock
   bIsVariable=true
   parent=NonEGSLink
   widget=Text_MoreInfo_1
   properties {
    bOverrideDefaultStyle=true
    lineHeightPercentage=0.875
    mobileTextBlockScale=1.2
    text="To view which platforms you are currently subscribed on, visit: "
    defaultTextStyleOverride {
     overflowPolicy=Clip
     transformPolicy=None
     colorAndOpacity {colorUseRule=UseColor_Specified, specifiedColor=<rgba 0.5776 0.5776 0.5776 1>}
     font {
      bForceMonospaced=false
      bMaterialIsStencil=false
      fontMaterial=null
      fontObject=&HeadingNow
      letterSpacing=0
      monospacedWidth=1
      size=15
      skewAmount=0
      typefaceFontName=74 Regular
      outlineSettings {
       bApplyOutlineToDropShadows=false
       bMiteredCorners=false
       bSeparateFillAlpha=false
       outlineMaterial=null
       outlineSize=0
       outlineColor=<rgba 0 0 0 1>
      }
     }
     highlightColor {colorUseRule=UseColor_Specified, specifiedColor=<rgba 0 0 0 1>}
     highlightShape {
      bIsDynamicallyLoaded=false
      drawAs=Image
      imageType=NoImage
      mirroring=NoMirror
      resourceName=None
      resourceObject=null
      tiling=NoTile
      imageSize=<32 32>
      margin=<ltrb 0 0 0 0>
      outlineSettings {
       bUseBrushTransparency=false
       roundingType=HalfHeightRadius
       width=0
       color {colorUseRule=UseColor_Specified, specifiedColor=<rgba 0 0 0 0>}
       cornerRadii=<0 0 0 0>
      }
      tintColor {colorUseRule=UseColor_Specified, specifiedColor=<rgba 1 1 1 1>}
      uVRegion {bIsValid=false, max=<0 0>, min=<0 0>}
     }
     selectedBackgroundColor {colorUseRule=UseColor_Foreground, specifiedColor=<rgba 1 0 1 1>}
     shadowColorAndOpacity=<rgba 0 0 0 1>
     shadowOffset=<0 0>
     strikeBrush {
      bIsDynamicallyLoaded=false
      drawAs=Image
      imageType=NoImage
      mirroring=NoMirror
      resourceName=None
      resourceObject=null
      tiling=NoTile
      imageSize=<32 32>
      margin=<ltrb 0 0 0 0>
      outlineSettings {
       bUseBrushTransparency=false
       roundingType=HalfHeightRadius
       width=0
       color {colorUseRule=UseColor_Specified, specifiedColor=<rgba 0 0 0 0>}
       cornerRadii=<0 0 0 0>
      }
      tintColor {colorUseRule=UseColor_Specified, specifiedColor=<rgba 1 1 1 1>}
      uVRegion {bIsValid=false, max=<0 0>, min=<0 0>}
     }
     underlineBrush {
      bIsDynamicallyLoaded=false
      drawAs=Image
      imageType=NoImage
      mirroring=NoMirror
      resourceName=None
      resourceObject=null
      tiling=NoTile
      imageSize=<32 32>
      margin=<ltrb 0 0 0 0>
      outlineSettings {
       bUseBrushTransparency=false
       roundingType=HalfHeightRadius
       width=0
       color {colorUseRule=UseColor_Specified, specifiedColor=<rgba 0 0 0 0>}
       cornerRadii=<0 0 0 0>
      }
      tintColor {colorUseRule=UseColor_Specified, specifiedColor=<rgba 1 1 1 1>}
      uVRegion {bIsValid=false, max=<0 0>, min=<0 0>}
     }
    }
   }
   slot {class=HorizontalBoxSlot, properties {horizontalAlignment=HAlign_Left, verticalAlignment=VAlign_Center, padding=<ltrb 0 0 0 2>}}
  }
  Button_OpenCheckCrewURL {
   widgetClassPath=&UserActionMenuInputButton
   bIsVariable=true
   parent=NonEGSLink
   widget=Button_OpenCheckCrewURL
   properties {
    onButtonBaseClicked=(&self:WidgetTree.Button_OpenCheckCrewURL.HandleOnButtonClicked)
    onButtonBaseDoubleClicked=(&self:WidgetTree.Button_OpenCheckCrewURL.HandleOnButtonDoubleClicked)
    onButtonBaseHovered=(&self:WidgetTree.Button_OpenCheckCrewURL.HandleOnButtonHovered)
    onButtonBaseUnhovered=(&self:WidgetTree.Button_OpenCheckCrewURL.HandleOnButtonUnhovered)
    onSelectedChangedBase=(&self:WidgetTree.Button_OpenCheckCrewURL.HandleOnSelectedChanged)
    triggeringInputAction {dataTable=&InputActionDataTable, rowName=PriceAcknowledge_MoreInfo}
   }
   slot {class=HorizontalBoxSlot, properties {horizontalAlignment=HAlign_Left, verticalAlignment=VAlign_Center}}
  }
  Button_TextURL_1 {
   widgetClassPath=&RichTextButton
   bIsVariable=true
   parent=NonEGSLink
   widget=Button_TextURL_1
   properties {
    auto wrap text=false
    button text=https://fn.gg/check-crew
    fontSize=16
    fontSizeOverride=true
    onButtonBaseClicked=(&self:WidgetTree.Button_TextURL_1.HandleOnButtonClicked)
    onButtonBaseDoubleClicked=(&self:WidgetTree.Button_TextURL_1.HandleOnButtonDoubleClicked)
    onButtonBaseHovered=(&self:WidgetTree.Button_TextURL_1.HandleOnButtonHovered)
    onButtonBaseUnhovered=(&self:WidgetTree.Button_TextURL_1.HandleOnButtonUnhovered)
    onSelectedChangedBase=(&self:WidgetTree.Button_TextURL_1.HandleOnSelectedChanged)
   }
   slot {class=HorizontalBoxSlot, properties {horizontalAlignment=HAlign_Left, verticalAlignment=VAlign_Center}}
  }
  HardwareVisibilityBorder_TouchClose {
   widgetClassPath=&CommonHardwareVisibilityBorder
   parent=Overlay_Root
   widget=HardwareVisibilityBorder_TouchClose
   properties {visibilityQuery {
     autoDescription=" ANY( Input.Touch )"
     tokenStreamVersion=0
     userDescription=""
     queryTokenStream [0,1,1,1,0]
     tagDictionary [{tagName=Input.Touch}]
    }}
   slot {class=OverlaySlot, properties {horizontalAlignment=HAlign_Right, padding=<ltrb 0 32 32 0>}}
  }
  SafeZone_TouchClose {
   widgetClassPath=&SafeZone
   bIsVariable=true
   parent=HardwareVisibilityBorder_TouchClose
   widget=SafeZone_TouchClose
   slot {class=BorderSlot, properties {padding=<ltrb 0 0 0 0>}}
  }
  Button_TouchClose {
   widgetClassPath=/FortUILibrary/Foundation/Buttons/CloseButton
   bIsVariable=true
   parent=SafeZone_TouchClose
   widget=Button_TouchClose
   properties {
    onButtonBaseClicked=(&self:WidgetTree.Button_TouchClose.HandleOnButtonClicked)
    onButtonBaseDoubleClicked=(&self:WidgetTree.Button_TouchClose.HandleOnButtonDoubleClicked)
    onButtonBaseHovered=(&self:WidgetTree.Button_TouchClose.HandleOnButtonHovered)
    onButtonBaseUnhovered=(&self:WidgetTree.Button_TouchClose.HandleOnButtonUnhovered)
    onSelectedChangedBase=(&self:WidgetTree.Button_TouchClose.HandleOnSelectedChanged)
   }
   slot {class=SafeZoneSlot}
  }
 ]
}
