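# Layout-less CommonActivatableWidget base for subscription modals, with child widgets expected to supply animations, sound and widget references.
# OnActivatedDefault plays the show-up animation and sound and sets the scroll widget.
# Backout is ignored mid-animation, otherwise it plays DismissAnimation and deactivates when that finishes, or immediately if none is set.
# Platform helpers rebuild text entries in an overridable panel and set the QR image URL, showing its container only when that URL is empty.
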WidgetBlueprint /Crew/UI/Subscription/SubscriptionModalBase
graphs [
 EventGraph {nodes [
   TriggerModalBackout() {is=title K2Node_CustomEvent}
   > n1 = Branch(Condition=IsAnyAnimationPlaying!(self:User Widget Object Reference) -> (ReturnValue:bool))
   n2 = Get DismissAnimation() -> (DismissAnimation:Widget Animation Object Reference)
   PlayAnimationForward(execute=@n1.else, InAnimation=@n2)
   > n3 = Branch(Condition=IsValid(Object=@n2))
   DeactivateWidget(execute=[@n5.then, @n3.else])
   n4 = Event OnAnimationFinished!() -> (OutputDelegate, then, Animation:Widget Animation Object Reference) {is=title K2Node_Event}
   > n5 = Branch(Condition="Equal (Object)"(A=@n2, B=@n4.Animation))
  ]}
 GetActivePlatformWidget {nodes [
   n1 = GetActivePlatformWidget!() -> (then, IsNotOnActivePlatformWidget:Widget Object Reference, IsOnActivePlatformWidget:Widget Object Reference, IsOnActivePlatform:bool=~) {is=title K2Node_FunctionEntry}
   > Return Node!(
    execute
    Widget:Widget Object Reference=Select(Option 0:Widget Object Reference=@n1, Option 1:Widget Object Reference=@n1.IsOnActivePlatformWidget, Index=@n1.IsOnActivePlatform) -> (ReturnValue:Widget Object Reference)
   )
  ]}
 GetPlatformSpecificEntriesWidget {nodes [
   GetPlatformSpecificEntriesWidget() {is=title K2Node_FunctionEntry}
   > Return Node!(execute, EntriesWidget:Panel Widget Object Reference)
  ]}
 OnActivatedDefault {nodes [
   OnActivatedDefault() {is=title K2Node_FunctionEntry}
   > PlayAnimationForward(InAnimation=Get ShowUpAnimation() -> (ShowUpAnimation:Widget Animation Object Reference))
   > SetScrollWidget!(execute, ToScroll:Widget Object Reference=Get DefaultScrollWidget() -> (DefaultScrollWidget:Widget Object Reference)) -> (then)
   > SpawnSound2D!(
    execute
    Sound:Sound Base Object Reference=Get ScreenAppearSound() -> (ScreenAppearSound:Sound Base Object Reference)
    VolumeMultiplier:float=1
    PitchMultiplier:float=1
    StartTime:float=0
    ConcurrencySettings:Sound Concurrency Object Reference
    bPersistAcrossLevelTransition:bool=false
    bAutoDestroy:bool=true
   ) -> (then, ReturnValue:Audio Component Object Reference)
  ]}
 SetPlatformSpecificEntries {nodes [
   n1 = SetPlatformSpecificEntries!() -> (then, Entries:Array of Texts) {is=title K2Node_FunctionEntry}
   n2 = GetPlatformSpecificEntriesWidget!(self) -> (EntriesWidget:Panel Widget Object Reference) {is=title K2Node_CallFunction}
   > ClearChildren(self=@n2)
   > n3 = For Each Loop(Array:Array of Texts=@n1) -> (Array Element:Text)
   > n4 = Create Widget!(
    execute:Exec
    Class:User Widget Class Reference=&/SubscriptionModalEntry.SubscriptionModalEntry_C
    OwningPlayer:Player Controller Object Reference
    text:Text=@n3
   ) -> (then:Exec, ReturnValue:Subscription Modal Entry Object Reference)
   > AddChild(self=@n2, Content=@n4)
  ]}
 SetPlatformSpecificQRCode {nodes [
   n1 = SetPlatformSpecificQRCode!() -> (then, QRMediaURL:String) {is=title K2Node_FunctionEntry}
   > SetMediaURL!(execute, self:Epic CMSImage Object Reference=Get QRImageWidget() -> (QRImageWidget:Epic CMSImage Object Reference), MediaUrl:String=@n1) -> (then)
   > SetVisibility(
    self=Get QRImageContainer() -> (QRImageContainer:Widget Object Reference)
    InVisibility=Select(Option 0:ESlateVisibility Enum=Hidden, Option 1:ESlateVisibility Enum=Visible, Index=IsEmpty(InString=@n1)) -> (ReturnValue:ESlateVisibility Enum)
   )
  ]}
]
variables [
 AnalyticsContext {type=string, default=""}
 DefaultScrollWidget {type=Widget, default=null}
 DismissAnimation {type=WidgetAnimation, default=null}
 ModalId {type=string, default=""}
 QRImageContainer {type=Widget, default=null}
 QRImageWidget {type=EpicCMSImage, default=null}
 ScreenAppearSound {type=SoundBase, default=null}
 ShowUpAnimation {type=WidgetAnimation, default=null}
]
widgets {info {inheritedWidgetCount=0, parentClass=&CommonActivatableWidget, rootWidgetClass=null, widgetCount=0}, widgets []}
