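# Crew subscription widget that shows the Crew logo badge above a centered "Subscriber since" label and start date.
# HandleStartDate formats a given UTC DateTime with short date and time styles in the GMT time zone, appends "GMT" and writes it to the StartDate text.
# On Construct, SetCompactUIStyle checks IsMobileGame and overrides the vertical slot padding of both text blocks for a compact layout.
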WidgetBlueprint /Crew/UI/Subscription/WBP_SubscriberSince
graphs [
 EventGraph {nodes [
   Event Construct()
   > SetCompactUIStyle() {is=title K2Node_CallFunction}
   Event PreConstruct()
   Event Tick()
  ]}
 HandleStartDate {nodes [
   n1 = HandleStartDate!() -> (then, StartDate:Date Time Structure) {is=title K2Node_FunctionEntry}
   > Set Text(
    Text=Format Text!(
     Format="LOCTEXT(\"{datetime} GMT\")"
     datetime:Text="As DateTime (from UTC)"!(
      InDateTime:"Date Time Structure (by ref)"=@n1
      InTimeZone:String=GMT
      InDateStyle:EDateTimeStyle Enum=Short
      InTimeStyle:EDateTimeStyle Enum=Short
     ) -> (ReturnValue:Text)
    ) -> (Result)
    self=Get StartDate() -> (StartDate:Common Text Object Reference)
   )
  ]}
 SetCompactUIStyle {nodes [
   SetCompactUIStyle() {is=title K2Node_FunctionEntry}
   > Branch(Condition=IsMobileGame())
   > SetPadding!(
    execute
    self:Vertical Box Slot Object Reference=SlotAsVerticalBoxSlot(Widget=Get Since() -> (Since:Common Text Object Reference))
    InPadding_Left:float=0
    InPadding_Top:float=15
    InPadding_Right:float=0
    InPadding_Bottom:float=0
   ) -> (then)
   > SetPadding(self=SlotAsVerticalBoxSlot(Widget=Get StartDate()), InPadding_Left=0, InPadding_Top=15, InPadding_Right=0, InPadding_Bottom=0)
  ]}
]
variables []
widgets {
 info {inheritedWidgetCount=0, parentClass=&UserWidget, rootWidgetClass=&VerticalBox, widgetCount=4}
 widgets [
  VerticalBox_23 {widgetClassPath=&VerticalBox, widget=VerticalBox_23}
  Badge {
   widgetClassPath=&Image
   bIsVariable=true
   parent=VerticalBox_23
   widget=Badge
   properties {
    visibility=HitTestInvisible
    brush {
     bIsDynamicallyLoaded=false
     drawAs=Image
     imageType=FullColor
     mirroring=NoMirror
     resourceName=None
     resourceObject=/Game/Athena/UI/Logos/Textures/M_UI_CrewLogo_Large
     tiling=NoTile
     imageSize=<368 110>
     margin=<ltrb 0 0 0 0>
     outlineSettings {
      bUseBrushTransparency=false
      roundingType=HalfHeightRadius
      width=0
      color {colorUseRule=UseColor_Specified, specifiedColor=<rgba 0 0 0 0>}
      cornerRadii=<0 0 0 0>
     }
     tintColor {colorUseRule=UseColor_Specified, specifiedColor=<rgba 1 1 1 1>}
     uVRegion {bIsValid=false, max=<0 0>, min=<0 0>}
    }
   }
   slot {class=VerticalBoxSlot, properties {horizontalAlignment=HAlign_Center, verticalAlignment=VAlign_Top}}
  }
  Since {
   widgetClassPath=&CommonTextBlock
   bIsVariable=true
   parent=VerticalBox_23
   widget=Since
   properties {
    text=Subscriber since
    colorAndOpacity {colorUseRule=UseColor_Specified, specifiedColor=<rgba 0.5776 0.5776 0.5776 1>}
    font {
     bForceMonospaced=false
     bMaterialIsStencil=false
     fontMaterial=null
     fontObject=&NotoSans
     letterSpacing=0
     monospacedWidth=1
     size=24
     skewAmount=0
     typefaceFontName=NotoSans-Regular
     outlineSettings {
      bApplyOutlineToDropShadows=false
      bMiteredCorners=false
      bSeparateFillAlpha=false
      outlineMaterial=null
      outlineSize=0
      outlineColor=<rgba 0 0 0 1>
     }
    }
    shadowOffset=<0 0>
   }
   slot {class=VerticalBoxSlot, properties {horizontalAlignment=HAlign_Center, padding=<ltrb 0 32 0 0>}}
  }
  StartDate {
   widgetClassPath=&CommonTextBlock
   bIsVariable=true
   parent=VerticalBox_23
   widget=StartDate
   properties {
    text=xx/xx/xx
    colorAndOpacity {colorUseRule=UseColor_Specified, specifiedColor=<rgba 0.5776 0.5776 0.5776 1>}
    font {
     bForceMonospaced=false
     bMaterialIsStencil=false
     fontMaterial=null
     fontObject=&NotoSans
     letterSpacing=0
     monospacedWidth=1
     size=24
     skewAmount=0
     typefaceFontName=NotoSans-Regular
     outlineSettings {
      bApplyOutlineToDropShadows=false
      bMiteredCorners=false
      bSeparateFillAlpha=false
      outlineMaterial=null
      outlineSize=0
      outlineColor=<rgba 0 0 0 1>
     }
    }
    shadowOffset=<0 0>
   }
   slot {class=VerticalBoxSlot, properties {horizontalAlignment=HAlign_Center, padding=<ltrb 0 5 0 0>}}
  }
 ]
}
