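# Fortnite Crew cancellation modal, built on SubscriptionModalBase, showing what players keep or lose and any steps needed to cancel.
# Pass lists are split by purchase context; error bullets replace platform steps and hide the cancel button.
# Cancel logs analytics, then, for Epic-billed subscribers, runs async CancelSubscription with a spinner and disabled buttons.
# Success plays the CancelSub animation, whose completion broadcasts CancellationModalComplete and deactivates the modal.
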WidgetBlueprint /Crew/UI/Subscription/SubscriptionCancellationInfoModal
animations [CancelSub {
  binding_count=7
  length_seconds=2.1901
  object=&self:CancelSub
  sequence {
   bindings [
    Div {
     id=#1
     possessed_class=Image
     tracks [
      {
       type=MovieSceneFloatTrack
       display_name=Render Opacity
       sections [{
         type=MovieSceneFloatSection
         end_frame=219
         start_frame=0
         channels [None {type=MovieSceneScriptingFloatChannel, keys [[20,0],[49,1],[190,1],[212,0]]}]
        }]
      }
      {
       type=MovieScene2DTransformTrack
       sections [{
         type=MovieScene2DTransformSection
         end_frame=219
         start_frame=0
         channels [
          Translation.X {type=MovieSceneScriptingFloatChannel, keys [[21,-120],[49,0],[190,0],[212,120]]}
          Translation.Y {type=MovieSceneScriptingFloatChannel, const [0,4]}
          Angle {type=MovieSceneScriptingFloatChannel}
          Scale.X {type=MovieSceneScriptingFloatChannel}
          Scale.Y {type=MovieSceneScriptingFloatChannel}
          Shear.X {type=MovieSceneScriptingFloatChannel}
          Shear.Y {type=MovieSceneScriptingFloatChannel}
         ]
        }]
      }
     ]
    }
    CancellationMessage {
     id=#2
     possessed_class=VerticalBox
     tracks [{
       type=MovieSceneByteTrack
       display_name=Visibility
       sections [{type=MovieSceneByteSection, channels [None {type=MovieSceneScriptingByteChannel, keys [[-0.0017,3],[219.0033,1]]}]}]
      }]
    }
    Success {
     id=#3
     possessed_class=CommonTextBlock
     tracks [
      {
       type=MovieScene2DTransformTrack
       sections [{
         type=MovieScene2DTransformSection
         channels [
          Translation.X {type=MovieSceneScriptingFloatChannel, keys [[38,-64],[54,0],[196.0017,0],[204.0017,64]]}
          Translation.Y {type=MovieSceneScriptingFloatChannel}
          Angle {type=MovieSceneScriptingFloatChannel}
          Scale.X {type=MovieSceneScriptingFloatChannel}
          Scale.Y {type=MovieSceneScriptingFloatChannel}
          Shear.X {type=MovieSceneScriptingFloatChannel}
          Shear.Y {type=MovieSceneScriptingFloatChannel}
         ]
        }]
      }
      {
       type=MovieSceneFloatTrack
       display_name=Render Opacity
       sections [{type=MovieSceneFloatSection, channels [None {type=MovieSceneScriptingFloatChannel, keys [[38,0],[54,1],[196.0017,1],[204.0017,0]]}]}]
      }
     ]
    }
    Sorry {
     id=#4
     possessed_class=CommonTextBlock
     tracks [
      {
       type=MovieScene2DTransformTrack
       sections [{
         type=MovieScene2DTransformSection
         channels [
          Translation.X {type=MovieSceneScriptingFloatChannel, keys [[46,-64],[62,0],[192,0],[200,64]]}
          Translation.Y {type=MovieSceneScriptingFloatChannel}
          Angle {type=MovieSceneScriptingFloatChannel}
          Scale.X {type=MovieSceneScriptingFloatChannel}
          Scale.Y {type=MovieSceneScriptingFloatChannel}
          Shear.X {type=MovieSceneScriptingFloatChannel}
          Shear.Y {type=MovieSceneScriptingFloatChannel}
         ]
        }]
      }
      {
       type=MovieSceneFloatTrack
       display_name=Render Opacity
       sections [{type=MovieSceneFloatSection, channels [None {type=MovieSceneScriptingFloatChannel, keys [[46,0],[62,1],[192,1],[200,0]]}]}]
      }
     ]
    }
    CancellationBg {
     id=#5
     possessed_class=Image
     tracks [{
       type=MovieSceneFloatTrack
       display_name=Render Opacity
       sections [{
         type=MovieSceneFloatSection
         end_frame=218
         start_frame=-1
         channels [None {type=MovieSceneScriptingFloatChannel, keys [[-0.0017,0],[38,1],[192,1],[218,0]]}]
        }]
      }]
    }
    LogoBadgeSmall {
     id=#6
     possessed_class=SizeBox
     tracks [
      {
       type=MovieScene2DTransformTrack
       sections [{
         type=MovieScene2DTransformSection
         channels [
          Translation.X {type=MovieSceneScriptingFloatChannel, keys [[30,-256],[46,0],[200,0],[208,256]]}
          Translation.Y {type=MovieSceneScriptingFloatChannel}
          Angle {type=MovieSceneScriptingFloatChannel}
          Scale.X {type=MovieSceneScriptingFloatChannel}
          Scale.Y {type=MovieSceneScriptingFloatChannel}
          Shear.X {type=MovieSceneScriptingFloatChannel}
          Shear.Y {type=MovieSceneScriptingFloatChannel}
         ]
        }]
      }
      {
       type=MovieSceneFloatTrack
       display_name=Render Opacity
       sections [{type=MovieSceneFloatSection, channels [None {type=MovieSceneScriptingFloatChannel, keys [[30,0],[46,1],[200,1],[208,0]]}]}]
      }
     ]
    }
    Dialog_Internal {
     id=#7
     possessed_class=WBP_UIKit_Dialog_Internal_C
     tracks [{
       type=MovieSceneFloatTrack
       display_name=Render Opacity
       sections [{
         type=MovieSceneFloatSection
         end_frame=51
         start_frame=-1
         channels [None {type=MovieSceneScriptingFloatChannel, keys [[-0.0017,1],[43,1],[51,0]]}]
        }]
      }]
    }
   ]
   display_rate {numerator=100}
  }
 }]
graphs [
 CancellationModalComplete {nodes [CancellationModalComplete() {is=title K2Node_FunctionEntry}]}
 EventGraph {nodes [
   n1 = ButtonClicked!() -> (OutputDelegate, then, ButtonIndex:int) {is=title K2Node_CustomEvent}
   > n2 = Switch on Int!(execute:Exec, Selection:int=@n1.ButtonIndex) -> (Default:Exec, 0:Exec, 1:Exec)
   CancelClicked(execute=@n2.0)
   TriggerModalBackout(execute=@n2.1, self:Subscription Modal Base Object Reference)
   CancelClicked() {is=title K2Node_CustomEvent}
   n3 = Get CrewVM() -> (CrewVM:Crew VM Object Reference)
   > SendSubscriptionAnalyticsEvent!(
    execute
    self:Crew VM Object Reference=@n3
    ContextName:String=Get AnalyticsContext() -> (AnalyticsContext:String)
    Interaction:String=CancelSubscriptionClicked
   ) -> (then)
   n4 = ActiveSubscriptionPlatform!(self:Crew VM Object Reference=@n3) -> (ReturnValue:EAppStore Enum=DebugStore)
   > Branch(Condition=AND Boolean(
     A=IsSubscribed!(self:Crew VM Object Reference=@n3) -> (ReturnValue:bool)
     B="Equal (Enum)"(A:EAppStore Enum=@n4, B:EAppStore Enum=EpicPurchasingService)
    ))
   > SetCancelOperationCycle!(execute, self, Starting:bool=true) -> (then)
   > n5 = CancelSubscription!(execute:Exec) -> (then:Exec, OnOperationCompleted:Exec, Success:bool=~) {is=title K2Node_AsyncAction}
   SetCancelOperationCycle(execute=@n5.OnOperationCompleted, Starting=false)
   > Branch(Condition=@n5)
   n6 = Get CancelSub() -> (CancelSub:Widget Animation Object Reference)
   > PlayAnimationForward(InAnimation=@n6)
   Event PreConstruct()
   > n7 = Construct Object from Class(Class=/Crew/UI/Subscription/Data/DialogVM_Cancel.DialogVM_Cancel_C) -> (ReturnValue:Dialog VM Cancel Object Reference)
   > SetDialogData!(
    execute
    self:WBP UIKit Dialog Internal Object Reference=Get Dialog_Internal() -> (Dialog_Internal:WBP UIKit Dialog Internal Object Reference)
    Data:UI Kit Dialog Viewmodel Object Reference=@n7
   ) -> (then)
   n8 = Event OnAnimationFinished!() -> (OutputDelegate, then, Animation:Widget Animation Object Reference) {is=title K2Node_Event}
   > Branch(Condition="Equal (Object)"(A=@n6, B=@n8.Animation))
   > Call Cancellation Modal Complete!(execute:Exec, self:Subscription Cancellation Info Modal Object Reference) -> (then:Exec) {is=title K2Node_CallDelegate}
   > DeactivateWidget()
   Event On Activated()
   > OnActivatedDefault(self:Subscription Modal Base Object Reference)
   > SetScrollWidget!(execute, ToScroll:Widget Object Reference=Get PrimaryScrollBox() -> (PrimaryScrollBox:Scroll Box Object Reference)) -> (then)
   > UpdatePostCancellationText()
   > Bind Event to Dialog Button Clicked Dispatcher(self:WBP UIKit Dialog Internal Object Reference=Get Dialog_Internal(), Delegate=Create Event()) {is=title K2Node_AddDelegate}
   > Set RenderOpacity!(execute, RenderOpacity:float=1, self:Widget Object Reference=Get Dialog_Internal()) -> (then, Output_Get:float)
   Event Construct()
   > SetCompactUIStyle() {is=title K2Node_CallFunction}
   Event On Deactivated()
   > Unbind Event from Dialog Button Clicked Dispatcher(self:WBP UIKit Dialog Internal Object Reference=Get Dialog_Internal(), Delegate=Create Event()) {is=title K2Node_RemoveDelegate}
  ]}
 FilterPassesOnPurchaseContext {nodes [
   n1 = FilterPassesOnPurchaseContext!() -> (then, PurchaseContext:ESeasonPassPurchaseContext Enum=None, SeasonPasses:Array of Season Pass VM Object References) {is=title K2Node_FunctionEntry}
   > n2 = For Each Loop(Array:Array of Season Pass VM Object References=@n1.SeasonPasses) -> (Array Element:Season Pass VM Object Reference)
   Return Node!(
    execute=@n2.Completed
    Result:Array of Season Pass VM Object References=Get FilteredPasses() -> (FilteredPasses:Array of Season Pass VM Object References)
   )
   n3 = Get PurchaseContext() -> (PurchaseContext:ESeasonPassPurchaseContext Enum=None)
   > Branch(Condition="Equal (Byte)"(
     A:Byte=Get PurchaseContext!(self:Season Pass VM Object Reference=@n2) -> (PurchaseContext=None)
     B:int="To Integer (Byte)"!(InByte:Byte=@n3) -> (ReturnValue:int=0)
    ) -> (ReturnValue:bool=~) {is=title K2Node_PromotableOperator})
   > Add(TargetArray:Array of Season Pass VM Object References=Get FilteredPasses(), NewItem:"Season Pass VM Object Reference (by ref)"=@n2) {is=title K2Node_CallArrayFunction}
  ]}
 GetActivePlatformWidget {nodes [
   n1 = GetActivePlatformWidget!() -> (then, IsNotOnActivePlatformWidget:Widget Object Reference, IsOnActivePlatformWidget:Widget Object Reference, IsOnActivePlatform:bool) {is=title K2Node_FunctionEntry}
   n2 = Get VerticalBox_ErrorStateEntries() -> (VerticalBox_ErrorStateEntries:Vertical Box Object Reference)
   > Return Node!(
    execute
    Widget:Widget Object Reference=Select(
     Option 0:Widget Object Reference=Parent: GetActivePlatformWidget!(
      IsNotOnActivePlatformWidget:Widget Object Reference=@n1
      IsOnActivePlatformWidget:Widget Object Reference=@n1.IsOnActivePlatformWidget
      IsOnActivePlatform:bool=@n1.IsOnActivePlatform
     ) -> (Widget:Widget Object Reference) {is=title K2Node_CallParentFunction}
     Option 1:Widget Object Reference=Get VerticalBox_ErrorStateEntries()
     Index=HasAnyChildren!(self:Panel Widget Object Reference=@n2) -> (ReturnValue:bool)
    ) -> (ReturnValue:Widget Object Reference)
   )
  ]}
 GetCancelButtonVisibility {nodes [
   n1 = GetCancelButtonVisibility!() -> (then, CanCancelDirectly:bool=~) {is=title K2Node_FunctionEntry}
   n2 = NOT Boolean(A=HasAnyChildren(self=Get VerticalBox_ErrorStateEntries()))
   > Return Node!(
    execute
    Result:ESlateVisibility Enum=Select(Option 0:ESlateVisibility Enum=Collapsed, Option 1:ESlateVisibility Enum=Visible, Index=AND Boolean(A=@n1, B=@n2)) -> (ReturnValue:ESlateVisibility Enum)
   )
  ]}
 GetKeepLosePassesText {nodes [
   n1 = GetKeepLosePassesText!() -> (then, Passes:Array of Season Pass VM Object References) {is=title K2Node_FunctionEntry}
   > Set PassesList(PassesList:Text) -> (Output_Get:Text)
   > n2 = For Each Loop(Array=@n1)
   Return Node(
    execute=@n2.Completed
    Result:Text=Format Text!(
     Format="LOCTEXT(\"Ability to claim premium rewards from <text color=\\\"FFFFFFFF\\\" fontface=\\\"black\\\">{PassesText}</>\")"
     PassesText:Text=Get PassesList() -> (PassesList:Text)
    ) -> (Result)
   )
   n3 = "Equal (Integer)"(A=@n2.Array Index)
   > Set PassesList(PassesList=Format Text!(
     Format="LOCTEXT(\"{Original}{Joiner}{NewPassName}\")"
     Original:Text=Get PassesList()
     Joiner:Text=Select(Option 0:Text="LOCTEXT(\", \")", Option 1:Text, Index=@n3) -> (ReturnValue:Text)
     NewPassName:Text=GetLongDisplayName!(self:Season Pass VM Object Reference=@n2) -> (ReturnValue:Text)
    ) -> (Result))
  ]}
 GetPlatformSpecificEntriesWidget {nodes [
   GetPlatformSpecificEntriesWidget() {is=title K2Node_FunctionEntry}
   > Return Node!(
    execute
    EntriesWidget:Panel Widget Object Reference=Get VerticalBox_PlatformSpecificEntries() -> (VerticalBox_PlatformSpecificEntries:Vertical Box Object Reference)
   )
  ]}
 SetCancelButtonVisibility {nodes [
   n1 = SetCancelButtonVisibility!() -> (then, Vis:ESlateVisibility Enum=Visible) {is=title K2Node_FunctionEntry}
   > SetButtonVisibility!(execute, self:WBP UIKit Dialog Internal Object Reference=Get Dialog_Internal(), ButtonIndex:int=0, Visibility:ESlateVisibility Enum=@n1) -> (then)
  ]}
 SetCancelOperationCycle {nodes [
   n1 = SetCancelOperationCycle!() -> (then, Starting:bool=~) {is=title K2Node_FunctionEntry}
   > SetVisibility(
    self=Get Image_Spinner() -> (Image_Spinner:Image Object Reference)
    InVisibility=Select(Option 0:ESlateVisibility Enum=Collapsed, Option 1:ESlateVisibility Enum=Visible, Index=@n1) -> (ReturnValue:ESlateVisibility Enum)
   )
   n2 = Get ButtonWidgets!(self:WBP UIKit Dialog Internal Object Reference=Get Dialog_Internal()) -> (ButtonWidgets:Array of WBP UIKit Button CTA Base Object References)
   > SetIsEnabled(
    self=[
     "Get (a copy)"(Array:Array of WBP UIKit Button CTA Base Object References=@n2) -> (Output:WBP UIKit Button CTA Base Object Reference)
     "Get (a copy)"(Array=@n2, Dimension 1=1)
    ]
    bInIsEnabled=NOT Boolean(A=@n1)
   )
  ]}
 SetCancellabilityInformation {nodes [
   n1 = SetCancellabilityInformation!() -> (then, EndDate:Text) {is=title K2Node_FunctionEntry}
   > SetText(
    self=Get SubEndDateText() -> (SubEndDateText:Common Rich Text Block Object Reference)
    InText=Select(
     Option 0:Text=Format Text!(
      Format="LOCTEXT(\"Subscription will be effective until the end of your current billing period: <text color=\\\"FFFFFFFF\\\" fontface=\\\"black\\\">{date}</>\")"
      date:Text=@n1
     ) -> (Result)
     Option 1:Text="LOCTEXT(\"Subscription will be effective until the end of your current billing period.\")"
     Index=TextIsEmpty(InText=@n1)
    ) -> (ReturnValue:Text)
   )
  ]}
 SetCompactUIStyle {nodes [
   SetCompactUIStyle() {is=title K2Node_FunctionEntry}
   > Branch(Condition=IsUsingCompactUI())
   > SetPadding!(
    execute
    self:Scroll Box Slot Object Reference=SlotAsScrollBoxSlot!(Widget:Widget Object Reference=Get ContentVB() -> (ContentVB:Vertical Box Object Reference)) -> (ReturnValue:Scroll Box Slot Object Reference)
    InPadding_Left:float=0
    InPadding_Top:float=0
    InPadding_Right:float=20
    InPadding_Bottom:float=0
   ) -> (then)
  ]}
 SetErrorStateBulletPoints {nodes [
   n1 = SetErrorStateBulletPoints!() -> (then, Array:Array of Texts) {is=title K2Node_FunctionEntry}
   > ClearChildren(self=Get VerticalBox_ErrorStateEntries())
   > n2 = For Each Loop(Array:Array of Texts=@n1) -> (Array Element:Text)
   > n3 = Create Widget!(
    execute:Exec
    Class:User Widget Class Reference=Get Class(Object=Get SubscriptionModalEntry_8() -> (SubscriptionModalEntry_8:Subscription Modal Entry Object Reference)) -> (ReturnValue:Subscription Modal Entry Class Reference)
    OwningPlayer:Player Controller Object Reference=GetOwningPlayer()
    text:Text=@n2
   ) -> (then:Exec, ReturnValue:Subscription Modal Entry Object Reference)
   > AddChildToVerticalBox(self=Get VerticalBox_ErrorStateEntries(), Content=@n3)
  ]}
 UpdatePostCancellationText {nodes [
   UpdatePostCancellationText() {is=title K2Node_FunctionEntry}
   n1 = FilterPassesOnPurchaseContext!(
    self
    PurchaseContext:ESeasonPassPurchaseContext Enum=Purchase
    SeasonPasses:Array of Season Pass VM Object References=Get SeasonPasses!(self:Season Pass System VM Object Reference=Get SeasonPassSystemVM() -> (SeasonPassSystemVM:Season Pass System VM Object Reference)) -> (SeasonPasses:Array of Season Pass VM Object References)
   ) -> (Result:Array of Season Pass VM Object References) {is=title K2Node_CallFunction}
   n2 = Get SubscriptionModalEntry_7() -> (SubscriptionModalEntry_7:Subscription Modal Entry Object Reference)
   > SetText(
    self:Subscription Modal Entry Object Reference=@n2
    InText:Text=GetKeepLosePassesText!(self, Passes:Array of Season Pass VM Object References=@n1) -> (Result:Text) {is=title K2Node_CallFunction}
   )
   > SetVisibility(
    self=@n2
    InVisibility=Select(
     Option 0:ESlateVisibility Enum=Visible
     Option 1:ESlateVisibility Enum=Collapsed
     Index=Is Empty!(TargetArray:Array of Season Pass VM Object References=@n1) -> (ReturnValue:bool)
    ) -> (ReturnValue:ESlateVisibility Enum)
   )
   n3 = FilterPassesOnPurchaseContext(PurchaseContext=Subscription, SeasonPasses=Get SeasonPasses(self=Get SeasonPassSystemVM()))
   > SetText(
    self=Get SubscriptionModalEntry_10() -> (SubscriptionModalEntry_10:Subscription Modal Entry Object Reference)
    InText=GetKeepLosePassesText(Passes=@n3)
   )
   > SetVisibility(
    self=Get VerticalBox_YoullLose() -> (VerticalBox_YoullLose:Vertical Box Object Reference)
    InVisibility=Select(Option 0=Visible, Option 1=Collapsed, Index=Is Empty(TargetArray=@n3))
   )
  ]}
 VM_HandleSubscriptionState {nodes [
   n1 = VM_HandleSubscriptionState!() -> (then, SubState:EMcpSubscriptionState Enum=Inactive) {is=title K2Node_FunctionEntry}
   > n2 = Switch on EMcpSubscriptionState!(execute, Selection:EMcpSubscriptionState Enum=@n1) -> (
    Inactive:Exec
    Active:Exec
    ActiveWithExternalBilling:Exec
    NotAutoRenewable:Exec
    Canceled:Exec
    PaymentProcessError:Exec
    PlatformAuthError:Exec
    RewardsDelayed:Exec
    Unavailable:Exec
    Unknown:Exec
   ) {is=title K2Node_SwitchEnum}
   SetVisibility(
    execute=[
     @n2.Active
     @n2.Canceled
     @n2.Inactive
     @n2.NotAutoRenewable
     @n2.PaymentProcessError
     @n2.PlatformAuthError
     @n2.RewardsDelayed
     @n2.Unavailable
     @n2.Unknown
    ]
    InVisibility=SelfHitTestInvisible
   )
   SetVisibility(execute=@n2.ActiveWithExternalBilling, self=Get SubEndDateText(), InVisibility=Collapsed)
  ]}
]
variables [CancellationModalComplete {}]
widgets {
 info {
  inheritedWidgetCount=0
  namedSlotCount=3
  parentClass=&/SubscriptionModalBase.SubscriptionModalBase_C
  rootWidgetClass=&Overlay
  widgetCount=41
 }
 widgets [
  Root {widgetClassPath=&Overlay, widget=Root}
  Dialog_Internal {
   widgetClassPath=/FortUILibrary/Systems/Dialog/WBP_UIKit_Dialog_Internal
   bIsVariable=true
   parent=Root
   widget=Dialog_Internal
   properties {
    useFooter=true
    useText=None
    buttons {
     each Button Horizontal Fill=false
     each Button Min Width=157
     each Button Size=Medium
     focuse Button=0
     space Between Buttons=24
     cTA Buttons Types [Loud,Regular]
     group Placement {
      horizontal Alignment=HAlign_Fill
      vertical Alignment=VAlign_Bottom
      grid {column=1, column Span=0, layer=0, row=2, row Span=0, nudge=<0 0>}
      padding=<ltrb 80 48 80 48>
     }
    }
    dialogSIze {
     height Override {enable=true, value=895}
     max Aspect Ratio {enable=false, value=0}
     max Desired Height {enable=true, value=800}
     max Desired Width {enable=false, value=0}
     min Aspect Ratio {enable=false, value=0}
     min Desired Height {enable=false, value=0}
     min Desired Width {enable=false, value=0}
     width Override {enable=true, value=1490}
    }
    titleProperties {
     auto Wrap Text=true
     justification=Left
     line Height Percentage=1
     mobile Font Size Multiplier=1.3
     transform Policy=ToUpper
     wrap Text At=808
     color {colorUseRule=UseColor_Specified, specifiedColor=<rgba 1 1 1 1>}
     font {
      bForceMonospaced=false
      bMaterialIsStencil=false
      fontMaterial=null
      fontObject=&HeadingNow
      letterSpacing=0
      monospacedWidth=1
      size=25.5
      skewAmount=0
      typefaceFontName=86 Bold Italic
      outlineSettings {
       bApplyOutlineToDropShadows=false
       bMiteredCorners=false
       bSeparateFillAlpha=false
       outlineMaterial=null
       outlineSize=0
       outlineColor=<rgba 0 0 0 1>
      }
     }
     placement {
      horizontal Alignment=HAlign_Fill
      vertical Alignment=VAlign_Top
      grid {column=1, column Span=0, layer=0, row=0, row Span=0, nudge=<0 0>}
      padding=<ltrb 80 80 80 32>
     }
    }
   }
   slot {class=OverlaySlot, properties {horizontalAlignment=HAlign_Fill, verticalAlignment=VAlign_Fill}}
  }
  Overlay_1 {widgetClassPath=&Overlay, namedSlotHost=Dialog_Internal, widget=Overlay_1}
  Content {
   widgetClassPath=&HorizontalBox
   parent=Overlay_1
   widget=Content
   slot {class=OverlaySlot, properties {horizontalAlignment=HAlign_Fill, verticalAlignment=VAlign_Fill, padding=<ltrb 80 0 80 0>}}
  }
  TextContent {
   widgetClassPath=&VerticalBox
   parent=Content
   widget=TextContent
   slot {class=HorizontalBoxSlot, properties {size {sizeRule=Fill, value=1}}}
  }
  PrimaryScrollBox {
   widgetClassPath=&ScrollBox
   bIsVariable=true
   parent=TextContent
   widget=PrimaryScrollBox
   properties {
    widgetBarStyle {
     thickness=16
     draggedThumbImage {
      bIsDynamicallyLoaded=false
      drawAs=RoundedBox
      imageType=FullColor
      mirroring=NoMirror
      resourceName=../../../Engine/Content/Slate/Common/Scrollbar_Thumb.png
      resourceObject=null
      tiling=NoTile
      imageSize=<12 12>
      margin=<ltrb 0.25 0.25 0.25 0.25>
      outlineSettings {
       bUseBrushTransparency=false
       roundingType=FixedRadius
       width=2
       color {colorUseRule=UseColor_Specified, specifiedColor=<rgba 0 0 0 0>}
       cornerRadii=<6 6 6 6>
      }
      tintColor {colorUseRule=UseColor_Specified, specifiedColor=<rgba 1 1 1 1>}
      uVRegion {bIsValid=false, max=<0 0>, min=<0 0>}
     }
     horizontalBackgroundImage {
      bIsDynamicallyLoaded=false
      drawAs=Image
      imageType=NoImage
      mirroring=NoMirror
      resourceName=None
      resourceObject=/Game/Athena/UI/ReplayBrowser/M_Ui_Skewed_Vfor9slices_BG_Separator_Inst
      tiling=NoTile
      imageSize=<0 0>
      margin=<ltrb 0 0 0 0>
      outlineSettings {
       bUseBrushTransparency=false
       roundingType=HalfHeightRadius
       width=0
       color {colorUseRule=UseColor_Specified, specifiedColor=<rgba 0 0 0 0>}
       cornerRadii=<0 0 0 0>
      }
      tintColor {colorUseRule=UseColor_Specified, specifiedColor=<rgba 0 0.2159 1 0.498>}
      uVRegion {bIsValid=false, max=<0 0>, min=<0 0>}
     }
     horizontalBottomSlotImage {
      bIsDynamicallyLoaded=false
      drawAs=NoDrawType
      imageType=FullColor
      mirroring=NoMirror
      resourceName=../../../Engine/Content/Slate/Common/Scrollbar_Background_Horizontal.png
      resourceObject=null
      tiling=NoTile
      imageSize=<8 8>
      margin=<ltrb 0 0 0 0>
      outlineSettings {
       bUseBrushTransparency=false
       roundingType=HalfHeightRadius
       width=0
       color {colorUseRule=UseColor_Specified, specifiedColor=<rgba 0 0 0 0>}
       cornerRadii=<0 0 0 0>
      }
      tintColor {colorUseRule=UseColor_Specified, specifiedColor=<rgba 1 1 1 1>}
      uVRegion {bIsValid=false, max=<0 0>, min=<0 0>}
     }
     horizontalTopSlotImage {
      bIsDynamicallyLoaded=false
      drawAs=NoDrawType
      imageType=FullColor
      mirroring=NoMirror
      resourceName=../../../Engine/Content/Slate/Common/Scrollbar_Background_Horizontal.png
      resourceObject=null
      tiling=NoTile
      imageSize=<8 8>
      margin=<ltrb 0 0 0 0>
      outlineSettings {
       bUseBrushTransparency=false
       roundingType=HalfHeightRadius
       width=0
       color {colorUseRule=UseColor_Specified, specifiedColor=<rgba 0 0 0 0>}
       cornerRadii=<0 0 0 0>
      }
      tintColor {colorUseRule=UseColor_Specified, specifiedColor=<rgba 1 1 1 1>}
      uVRegion {bIsValid=false, max=<0 0>, min=<0 0>}
     }
     hoveredThumbImage {
      bIsDynamicallyLoaded=false
      drawAs=RoundedBox
      imageType=FullColor
      mirroring=NoMirror
      resourceName=../../../Engine/Content/Slate/Common/Scrollbar_Thumb.png
      resourceObject=null
      tiling=NoTile
      imageSize=<12 12>
      margin=<ltrb 0.25 0.25 0.25 0.25>
      outlineSettings {
       bUseBrushTransparency=false
       roundingType=FixedRadius
       width=2
       color {colorUseRule=UseColor_Specified, specifiedColor=<rgba 0 0 0 0>}
       cornerRadii=<6 6 6 6>
      }
      tintColor {colorUseRule=UseColor_Specified, specifiedColor=<rgba 1 1 1 1>}
      uVRegion {bIsValid=false, max=<0 0>, min=<0 0>}
     }
     normalThumbImage {
      bIsDynamicallyLoaded=false
      drawAs=RoundedBox
      imageType=FullColor
      mirroring=NoMirror
      resourceName=../../../Engine/Content/Slate/Common/Scrollbar_Thumb.png
      resourceObject=&M_Ui_Skewed_Vfor9slices_BG_Separator_Inst
      tiling=NoTile
      imageSize=<12 12>
      margin=<ltrb 0.25 0.25 0.25 0.25>
      outlineSettings {
       bUseBrushTransparency=false
       roundingType=FixedRadius
       width=2
       color {colorUseRule=UseColor_Specified, specifiedColor=<rgba 0 0 0 0>}
       cornerRadii=<6 6 6 6>
      }
      tintColor {colorUseRule=UseColor_Specified, specifiedColor=<rgba 1 1 1 0.5>}
      uVRegion {bIsValid=false, max=<0 0>, min=<0 0>}
     }
     verticalBackgroundImage {
      bIsDynamicallyLoaded=false
      drawAs=Box
      imageType=FullColor
      mirroring=NoMirror
      resourceName=None
      resourceObject=/FortUILibrary/UIKit/ScrollBar/MaterialInstances/MI_UIKit_ScrollbarBackground
      tiling=NoTile
      imageSize=<12 12>
      margin=<ltrb 0.5 0.5 0.5 0.5>
      outlineSettings {
       bUseBrushTransparency=false
       roundingType=HalfHeightRadius
       width=0
       color {colorUseRule=UseColor_Specified, specifiedColor=<rgba 0 0 0 0>}
       cornerRadii=<0 0 0 0>
      }
      tintColor {colorUseRule=UseColor_Specified, specifiedColor=<rgba 1 1 1 0.2>}
      uVRegion {bIsValid=false, max=<0 0>, min=<0 0>}
     }
     verticalBottomSlotImage {
      bIsDynamicallyLoaded=false
      drawAs=NoDrawType
      imageType=FullColor
      mirroring=NoMirror
      resourceName=../../../Engine/Content/Slate/Common/Scrollbar_Background_Horizontal.png
      resourceObject=null
      tiling=NoTile
      imageSize=<8 8>
      margin=<ltrb 0 0 0 0>
      outlineSettings {
       bUseBrushTransparency=false
       roundingType=HalfHeightRadius
       width=0
       color {colorUseRule=UseColor_Specified, specifiedColor=<rgba 0 0 0 0>}
       cornerRadii=<0 0 0 0>
      }
      tintColor {colorUseRule=UseColor_Specified, specifiedColor=<rgba 1 1 1 1>}
      uVRegion {bIsValid=false, max=<0 0>, min=<0 0>}
     }
     verticalTopSlotImage {
      bIsDynamicallyLoaded=false
      drawAs=NoDrawType
      imageType=FullColor
      mirroring=NoMirror
      resourceName=../../../Engine/Content/Slate/Common/Scrollbar_Background_Horizontal.png
      resourceObject=null
      tiling=NoTile
      imageSize=<8 8>
      margin=<ltrb 0 0 0 0>
      outlineSettings {
       bUseBrushTransparency=false
       roundingType=HalfHeightRadius
       width=0
       color {colorUseRule=UseColor_Specified, specifiedColor=<rgba 0 0 0 0>}
       cornerRadii=<0 0 0 0>
      }
      tintColor {colorUseRule=UseColor_Specified, specifiedColor=<rgba 1 1 1 1>}
      uVRegion {bIsValid=false, max=<0 0>, min=<0 0>}
     }
    }
    widgetStyle {
     barThickness=9
     bottomShadowBrush {
      bIsDynamicallyLoaded=false
      drawAs=NoDrawType
      imageType=FullColor
      mirroring=NoMirror
      resourceName=../../../Engine/Content/Slate/Common/ScrollBoxShadowLeft.png
      resourceObject=null
      tiling=NoTile
      imageSize=<8 16>
      margin=<ltrb 1 0.5 0 0.5>
      outlineSettings {
       bUseBrushTransparency=false
       roundingType=HalfHeightRadius
       width=0
       color {colorUseRule=UseColor_Specified, specifiedColor=<rgba 0 0 0 0>}
       cornerRadii=<0 0 0 0>
      }
      tintColor {colorUseRule=UseColor_Specified, specifiedColor=<rgba 1 1 1 1>}
      uVRegion {bIsValid=false, max=<0 0>, min=<0 0>}
     }
     horizontalScrolledContentPadding=<ltrb 0 0 1 0>
     leftShadowBrush {
      bIsDynamicallyLoaded=false
      drawAs=NoDrawType
      imageType=FullColor
      mirroring=NoMirror
      resourceName=../../../Engine/Content/Slate/Common/ScrollBoxShadowLeft.png
      resourceObject=null
      tiling=NoTile
      imageSize=<8 16>
      margin=<ltrb 1 0.5 0 0.5>
      outlineSettings {
       bUseBrushTransparency=false
       roundingType=HalfHeightRadius
       width=0
       color {colorUseRule=UseColor_Specified, specifiedColor=<rgba 0 0 0 0>}
       cornerRadii=<0 0 0 0>
      }
      tintColor {colorUseRule=UseColor_Specified, specifiedColor=<rgba 1 1 1 1>}
      uVRegion {bIsValid=false, max=<0 0>, min=<0 0>}
     }
     rightShadowBrush {
      bIsDynamicallyLoaded=false
      drawAs=NoDrawType
      imageType=FullColor
      mirroring=NoMirror
      resourceName=../../../Engine/Content/Slate/Common/ScrollBoxShadowRight.png
      resourceObject=null
      tiling=NoTile
      imageSize=<8 16>
      margin=<ltrb 0 0.5 1 0.5>
      outlineSettings {
       bUseBrushTransparency=false
       roundingType=HalfHeightRadius
       width=0
       color {colorUseRule=UseColor_Specified, specifiedColor=<rgba 0 0 0 0>}
       cornerRadii=<0 0 0 0>
      }
      tintColor {colorUseRule=UseColor_Specified, specifiedColor=<rgba 1 1 1 1>}
      uVRegion {bIsValid=false, max=<0 0>, min=<0 0>}
     }
     topShadowBrush {
      bIsDynamicallyLoaded=false
      drawAs=NoDrawType
      imageType=FullColor
      mirroring=NoMirror
      resourceName=../../../Engine/Content/Slate/Common/ScrollBoxShadowLeft.png
      resourceObject=null
      tiling=NoTile
      imageSize=<8 16>
      margin=<ltrb 1 0.5 0 0.5>
      outlineSettings {
       bUseBrushTransparency=false
       roundingType=HalfHeightRadius
       width=0
       color {colorUseRule=UseColor_Specified, specifiedColor=<rgba 0 0 0 0>}
       cornerRadii=<0 0 0 0>
      }
      tintColor {colorUseRule=UseColor_Specified, specifiedColor=<rgba 1 1 1 1>}
      uVRegion {bIsValid=false, max=<0 0>, min=<0 0>}
     }
     verticalScrolledContentPadding=<ltrb 0 0 0 1>
    }
   }
   slot {class=VerticalBoxSlot, properties {size {sizeRule=Fill, value=1}}}
  }
  ContentVB {
   widgetClassPath=&VerticalBox
   bIsVariable=true
   parent=PrimaryScrollBox
   widget=ContentVB
   slot {class=ScrollBoxSlot, properties {padding=<ltrb 0 0 20 0>}}
  }
  SubEndDateText {
   widgetClassPath=&CommonRichTextBlock
   bIsVariable=true
   parent=ContentVB
   widget=SubEndDateText
   properties {
    autoWrapText=true
    bOverrideDefaultStyle=true
    lineHeightPercentage=0.875
    mobileTextBlockScale=1.2
    text="Subscription will be effective until the end of your current billing period: <text color=\"FFFFFFFF\" fontface=\"black\">{date}</>"
    defaultTextStyleOverride {
     overflowPolicy=Clip
     transformPolicy=None
     colorAndOpacity {colorUseRule=UseColor_Specified, specifiedColor=<rgba 1 1 1 1>}
     font {
      bForceMonospaced=false
      bMaterialIsStencil=false
      fontMaterial=null
      fontObject=&HeadingNow
      letterSpacing=0
      monospacedWidth=1
      size=15.75
      skewAmount=0
      typefaceFontName=74 Regular
      outlineSettings {
       bApplyOutlineToDropShadows=false
       bMiteredCorners=false
       bSeparateFillAlpha=false
       outlineMaterial=null
       outlineSize=0
       outlineColor=<rgba 0 0 0 1>
      }
     }
     highlightColor {colorUseRule=UseColor_Specified, specifiedColor=<rgba 0 0 0 1>}
     highlightShape {
      bIsDynamicallyLoaded=false
      drawAs=Image
      imageType=NoImage
      mirroring=NoMirror
      resourceName=None
      resourceObject=null
      tiling=NoTile
      imageSize=<32 32>
      margin=<ltrb 0 0 0 0>
      outlineSettings {
       bUseBrushTransparency=false
       roundingType=HalfHeightRadius
       width=0
       color {colorUseRule=UseColor_Specified, specifiedColor=<rgba 0 0 0 0>}
       cornerRadii=<0 0 0 0>
      }
      tintColor {colorUseRule=UseColor_Specified, specifiedColor=<rgba 1 1 1 1>}
      uVRegion {bIsValid=false, max=<0 0>, min=<0 0>}
     }
     selectedBackgroundColor {colorUseRule=UseColor_Foreground, specifiedColor=<rgba 1 0 1 1>}
     shadowColorAndOpacity=<rgba 0 0 0 1>
     shadowOffset=<0 0>
     strikeBrush {
      bIsDynamicallyLoaded=false
      drawAs=Image
      imageType=NoImage
      mirroring=NoMirror
      resourceName=None
      resourceObject=null
      tiling=NoTile
      imageSize=<32 32>
      margin=<ltrb 0 0 0 0>
      outlineSettings {
       bUseBrushTransparency=false
       roundingType=HalfHeightRadius
       width=0
       color {colorUseRule=UseColor_Specified, specifiedColor=<rgba 0 0 0 0>}
       cornerRadii=<0 0 0 0>
      }
      tintColor {colorUseRule=UseColor_Specified, specifiedColor=<rgba 1 1 1 1>}
      uVRegion {bIsValid=false, max=<0 0>, min=<0 0>}
     }
     underlineBrush {
      bIsDynamicallyLoaded=false
      drawAs=Image
      imageType=NoImage
      mirroring=NoMirror
      resourceName=None
      resourceObject=null
      tiling=NoTile
      imageSize=<32 32>
      margin=<ltrb 0 0 0 0>
      outlineSettings {
       bUseBrushTransparency=false
       roundingType=HalfHeightRadius
       width=0
       color {colorUseRule=UseColor_Specified, specifiedColor=<rgba 0 0 0 0>}
       cornerRadii=<0 0 0 0>
      }
      tintColor {colorUseRule=UseColor_Specified, specifiedColor=<rgba 1 1 1 1>}
      uVRegion {bIsValid=false, max=<0 0>, min=<0 0>}
     }
    }
   }
   slot {class=VerticalBoxSlot, properties {verticalAlignment=VAlign_Center, size {sizeRule=Fill, value=1}}}
  }
  VBox_PlatformSpecific {
   widgetClassPath=&VerticalBox
   bIsVariable=true
   parent=ContentVB
   widget=VBox_PlatformSpecific
   slot {class=VerticalBoxSlot, properties {padding=<ltrb 0 32 0 0>}}
  }
  Title {
   widgetClassPath=&CommonTextBlock
   parent=VBox_PlatformSpecific
   widget=Title
   properties {
    text=You'll need to:
    textTransformPolicy=ToUpper
    wrapTextAt=1152
    font {
     bForceMonospaced=false
     bMaterialIsStencil=false
     fontMaterial=null
     fontObject=&Burbank
     letterSpacing=0
     monospacedWidth=1
     size=30
     skewAmount=0
     typefaceFontName=BurbankRegular-Black
     outlineSettings {
      bApplyOutlineToDropShadows=false
      bMiteredCorners=false
      bSeparateFillAlpha=false
      outlineMaterial=null
      outlineSize=0
      outlineColor=<rgba 1 1 1 0>
     }
    }
    shadowOffset=<0 0>
   }
   slot {class=VerticalBoxSlot, properties {padding=<ltrb 0 0 0 16>}}
  }
  WidgetSwitcher_PlatformSpecific {
   widgetClassPath=&WidgetSwitcher
   bIsVariable=true
   parent=VBox_PlatformSpecific
   widget=WidgetSwitcher_PlatformSpecific
   slot {class=VerticalBoxSlot}
  }
  VerticalBox_PlatformSpecificEntries {
   widgetClassPath=&VerticalBox
   bIsVariable=true
   parent=WidgetSwitcher_PlatformSpecific
   widget=VerticalBox_PlatformSpecificEntries
   slot {class=WidgetSwitcherSlot}
  }
  SubscriptionModalEntry_5 {
   widgetClassPath=&/SubscriptionModalEntry
   bIsVariable=true
   parent=VerticalBox_PlatformSpecificEntries
   widget=SubscriptionModalEntry_5
   properties {text="[PH] Instructions 1"}
   slot {class=VerticalBoxSlot}
  }
  SubscriptionModalEntry_3 {
   widgetClassPath=&SubscriptionModalEntry
   bIsVariable=true
   parent=VerticalBox_PlatformSpecificEntries
   widget=SubscriptionModalEntry_3
   properties {text="[PH] Instructions 2"}
   slot {class=VerticalBoxSlot}
  }
  SubscriptionModalEntry_4 {
   widgetClassPath=&SubscriptionModalEntry
   bIsVariable=true
   parent=VerticalBox_PlatformSpecificEntries
   widget=SubscriptionModalEntry_4
   properties {text="[PH] Instructions 3"}
   slot {class=VerticalBoxSlot}
  }
  VerticalBox_WrongPlatform {
   widgetClassPath=&VerticalBox
   bIsVariable=true
   parent=WidgetSwitcher_PlatformSpecific
   widget=VerticalBox_WrongPlatform
   slot {class=WidgetSwitcherSlot}
  }
  SubscriptionModalEntry {
   widgetClassPath=&SubscriptionModalEntry
   bIsVariable=true
   parent=VerticalBox_WrongPlatform
   widget=SubscriptionModalEntry
   properties {text=Visit the platform on which you originally joined The Fortnite Crew.}
   slot {class=VerticalBoxSlot}
  }
  SubscriptionModalEntry_2 {
   widgetClassPath=&SubscriptionModalEntry
   bIsVariable=true
   parent=VerticalBox_WrongPlatform
   widget=SubscriptionModalEntry_2
   properties {text=Return to this screen.}
   slot {class=VerticalBoxSlot}
  }
  SubscriptionModalEntry_6 {
   widgetClassPath=&SubscriptionModalEntry
   bIsVariable=true
   parent=VerticalBox_WrongPlatform
   widget=SubscriptionModalEntry_6
   properties {text="Follow the final cancellation step(s)."}
   slot {class=VerticalBoxSlot}
  }
  VerticalBox_ErrorStateEntries {
   widgetClassPath=&VerticalBox
   bIsVariable=true
   parent=WidgetSwitcher_PlatformSpecific
   widget=VerticalBox_ErrorStateEntries
   slot {class=WidgetSwitcherSlot}
  }
  VerticalBox_YoullKeep {
   widgetClassPath=&VerticalBox
   parent=ContentVB
   widget=VerticalBox_YoullKeep
   slot {class=VerticalBoxSlot, properties {padding=<ltrb 0 32 0 0>}}
  }
  Title1 {
   widgetClassPath=&CommonTextBlock
   parent=VerticalBox_YoullKeep
   widget=Title1
   properties {
    text=You'll Keep:
    textTransformPolicy=ToUpper
    wrapTextAt=1152
    colorAndOpacity {colorUseRule=UseColor_Specified, specifiedColor=<rgba 0.0395 1 0.2307 1>}
    font {
     bForceMonospaced=false
     bMaterialIsStencil=false
     fontMaterial=null
     fontObject=&HeadingNow
     letterSpacing=0
     monospacedWidth=1
     size=30
     skewAmount=0
     typefaceFontName=76 Bold
     outlineSettings {
      bApplyOutlineToDropShadows=false
      bMiteredCorners=false
      bSeparateFillAlpha=false
      outlineMaterial=null
      outlineSize=0
      outlineColor=<rgba 1 1 1 0>
     }
    }
    shadowOffset=<0 0>
   }
   slot {class=VerticalBoxSlot, properties {padding=<ltrb 0 0 0 16>}}
  }
  SubscriptionModalEntry_8 {
   widgetClassPath=&SubscriptionModalEntry
   bIsVariable=true
   parent=VerticalBox_YoullKeep
   widget=SubscriptionModalEntry_8
   properties {text=Previously rewarded V-Bucks and Crew Packs}
   slot {class=VerticalBoxSlot}
  }
  SubscriptionModalEntry_9 {
   widgetClassPath=&SubscriptionModalEntry
   parent=VerticalBox_YoullKeep
   widget=SubscriptionModalEntry_9
   properties {text=Previously claimed rewards from all Passes}
   slot {class=VerticalBoxSlot}
  }
  SubscriptionModalEntry_1 {
   widgetClassPath=&SubscriptionModalEntry
   parent=VerticalBox_YoullKeep
   widget=SubscriptionModalEntry_1
   properties {text="All previously unlocked stages for your Crew items and should you decided to return, you will continue your progression where you left off"}
   slot {class=VerticalBoxSlot}
  }
  SubscriptionModalEntry_7 {
   widgetClassPath=&SubscriptionModalEntry
   bIsVariable=true
   parent=VerticalBox_YoullKeep
   widget=SubscriptionModalEntry_7
   properties {text="Ability to claim premium rewards from {List of the player's grant-based Passes}"}
   slot {class=VerticalBoxSlot}
  }
  VerticalBox_YoullLose {
   widgetClassPath=&VerticalBox
   bIsVariable=true
   parent=ContentVB
   widget=VerticalBox_YoullLose
   slot {class=VerticalBoxSlot, properties {padding=<ltrb 0 32 0 0>}}
  }
  Title2 {
   widgetClassPath=&CommonTextBlock
   parent=VerticalBox_YoullLose
   widget=Title2
   properties {
    text=You'll Lose:
    textTransformPolicy=ToUpper
    wrapTextAt=1152
    colorAndOpacity {colorUseRule=UseColor_Specified, specifiedColor=<rgba 1 0.147 0.1329 1>}
    font {
     bForceMonospaced=false
     bMaterialIsStencil=false
     fontMaterial=null
     fontObject=&HeadingNow
     letterSpacing=0
     monospacedWidth=1
     size=30
     skewAmount=0
     typefaceFontName=76 Bold
     outlineSettings {
      bApplyOutlineToDropShadows=false
      bMiteredCorners=false
      bSeparateFillAlpha=false
      outlineMaterial=null
      outlineSize=0
      outlineColor=<rgba 1 1 1 0>
     }
    }
    shadowOffset=<0 0>
   }
   slot {class=VerticalBoxSlot, properties {padding=<ltrb 0 0 0 16>}}
  }
  SubscriptionModalEntry_10 {
   widgetClassPath=&SubscriptionModalEntry
   bIsVariable=true
   parent=VerticalBox_YoullLose
   widget=SubscriptionModalEntry_10
   properties {text="Ability to claim premium rewards from {List of the player's access-based Passes}"}
   slot {class=VerticalBoxSlot}
  }
  Overlay_2 {widgetClassPath=&Overlay, namedSlotHost=Dialog_Internal, widget=Overlay_2}
  Image_Spinner {
   widgetClassPath=&Image
   bIsVariable=true
   parent=Overlay_2
   widget=Image_Spinner
   properties {
    visibility=Collapsed
    brush {
     bIsDynamicallyLoaded=false
     drawAs=Image
     imageType=NoImage
     mirroring=NoMirror
     resourceName=None
     resourceObject=/FortUILibrary/UIKit/Throbber/MI_UIKit_Monochrome_Throbber
     tiling=NoTile
     imageSize=<64 64>
     margin=<ltrb 0 0 0 0>
     outlineSettings {
      bUseBrushTransparency=false
      roundingType=HalfHeightRadius
      width=0
      color {colorUseRule=UseColor_Specified, specifiedColor=<rgba 0 0 0 0>}
      cornerRadii=<0 0 0 0>
     }
     tintColor {colorUseRule=UseColor_Specified, specifiedColor=<rgba 1 1 1 1>}
     uVRegion {bIsValid=false, max=<0 0>, min=<0 0>}
    }
   }
   slot {class=OverlaySlot, properties {verticalAlignment=VAlign_Bottom, padding=<ltrb 10 0 0 56>}}
  }
  ImageContent {widgetClassPath=&VerticalBox, namedSlotHost=Dialog_Internal, widget=ImageContent}
  WBP_SubscriberSince {
   widgetClassPath=&/WBP_SubscriberSince
   bIsVariable=true
   parent=ImageContent
   widget=WBP_SubscriberSince
   slot {class=VerticalBoxSlot, properties {verticalAlignment=VAlign_Center, padding=<ltrb 0 0 80 0>, size {sizeRule=Fill, value=1}}}
  }
  CancellationBg {
   widgetClassPath=&Image
   bIsVariable=true
   parent=Root
   widget=CancellationBg
   properties {
    renderOpacity=0
    visibility=SelfHitTestInvisible
    brush {
     bIsDynamicallyLoaded=false
     drawAs=Image
     imageType=FullColor
     mirroring=NoMirror
     resourceName=None
     resourceObject=/Crew/UI/Materials/M_UI_CrewUpsell_Landing_Sub
     tiling=NoTile
     imageSize=<32 32>
     margin=<ltrb 0 0 0 0>
     outlineSettings {
      bUseBrushTransparency=false
      roundingType=HalfHeightRadius
      width=0
      color {colorUseRule=UseColor_Specified, specifiedColor=<rgba 0 0 0 0>}
      cornerRadii=<0 0 0 0>
     }
     tintColor {colorUseRule=UseColor_Specified, specifiedColor=<rgba 1 1 1 1>}
     uVRegion {bIsValid=false, max=<0 0>, min=<0 0>}
    }
   }
   slot {class=OverlaySlot, properties {horizontalAlignment=HAlign_Fill, verticalAlignment=VAlign_Fill}}
  }
  CancellationMessage {
   widgetClassPath=&VerticalBox
   parent=Root
   widget=CancellationMessage
   properties {visibility=Collapsed}
   slot {class=OverlaySlot, properties {horizontalAlignment=HAlign_Center, verticalAlignment=VAlign_Center}}
  }
  LogoBadgeSmall {
   widgetClassPath=&SizeBox
   parent=CancellationMessage
   widget=LogoBadgeSmall
   properties {bOverride_HeightOverride=true, heightOverride=256}
   slot {class=VerticalBoxSlot, properties {verticalAlignment=VAlign_Center, size {sizeRule=Fill, value=1}}}
  }
  Overlay {
   widgetClassPath=&Overlay
   parent=LogoBadgeSmall
   widget=Overlay
   slot {class=SizeBoxSlot, properties {horizontalAlignment=HAlign_Center, verticalAlignment=VAlign_Center}}
  }
  Logo {
   widgetClassPath=&Image
   bIsVariable=true
   parent=Overlay
   widget=Logo
   properties {
    visibility=HitTestInvisible
    brush {
     bIsDynamicallyLoaded=false
     drawAs=Image
     imageType=FullColor
     mirroring=NoMirror
     resourceName=None
     resourceObject=/Game/Athena/UI/Logos/Textures/T_Crew_Workmark_L
     tiling=NoTile
     imageSize=<540 160>
     margin=<ltrb 0 0 0 0>
     outlineSettings {
      bUseBrushTransparency=false
      roundingType=HalfHeightRadius
      width=0
      color {colorUseRule=UseColor_Specified, specifiedColor=<rgba 0 0 0 0>}
      cornerRadii=<0 0 0 0>
     }
     tintColor {colorUseRule=UseColor_Specified, specifiedColor=<rgba 1 1 1 1>}
     uVRegion {bIsValid=false, max=<0 0>, min=<0 0>}
    }
   }
   slot {class=OverlaySlot, properties {horizontalAlignment=HAlign_Center, verticalAlignment=VAlign_Center}}
  }
  Div {
   widgetClassPath=&Image
   bIsVariable=true
   parent=CancellationMessage
   widget=Div
   properties {
    visibility=HitTestInvisible
    brush {
     bIsDynamicallyLoaded=false
     drawAs=Image
     imageType=NoImage
     mirroring=NoMirror
     resourceName=None
     resourceObject=null
     tiling=NoTile
     imageSize=<240 4>
     margin=<ltrb 0 0 0 0>
     outlineSettings {
      bUseBrushTransparency=false
      roundingType=HalfHeightRadius
      width=0
      color {colorUseRule=UseColor_Specified, specifiedColor=<rgba 0 0 0 0>}
      cornerRadii=<0 0 0 0>
     }
     tintColor {colorUseRule=UseColor_Specified, specifiedColor=<rgba 0.877 0.877 0.877 0.2>}
     uVRegion {bIsValid=false, max=<0 0>, min=<0 0>}
    }
   }
   slot {class=VerticalBoxSlot, properties {horizontalAlignment=HAlign_Center, verticalAlignment=VAlign_Center, padding=<ltrb 16 16 16 32>}}
  }
  Success {
   widgetClassPath=&CommonTextBlock
   parent=CancellationMessage
   widget=Success
   properties {
    text=You have successfully cancelled your subscription.
    textTransformPolicy=ToUpper
    colorAndOpacity {colorUseRule=UseColor_Specified, specifiedColor=<rgba 0.5776 0.5776 0.5776 1>}
    font {
     bForceMonospaced=false
     bMaterialIsStencil=false
     fontMaterial=null
     fontObject=&BurbankSmall
     letterSpacing=0
     monospacedWidth=1
     size=18
     skewAmount=0
     typefaceFontName=Black
     outlineSettings {
      bApplyOutlineToDropShadows=false
      bMiteredCorners=false
      bSeparateFillAlpha=false
      outlineMaterial=null
      outlineSize=0
      outlineColor=<rgba 0 0 0 1>
     }
    }
    shadowOffset=<0 0>
   }
   slot {class=VerticalBoxSlot, properties {horizontalAlignment=HAlign_Center}}
  }
  Sorry {
   widgetClassPath=&CommonTextBlock
   parent=CancellationMessage
   widget=Sorry
   properties {
    text="Sorry to see you go!"
    textTransformPolicy=ToUpper
    font {
     bForceMonospaced=false
     bMaterialIsStencil=false
     fontMaterial=null
     fontObject=&BurbankSmall
     letterSpacing=0
     monospacedWidth=1
     size=24
     skewAmount=0
     typefaceFontName=Black
     outlineSettings {
      bApplyOutlineToDropShadows=false
      bMiteredCorners=false
      bSeparateFillAlpha=false
      outlineMaterial=null
      outlineSize=0
      outlineColor=<rgba 0 0 0 1>
     }
    }
    shadowOffset=<0 0>
   }
   slot {class=VerticalBoxSlot, properties {horizontalAlignment=HAlign_Center, padding=<ltrb 0 16 0 0>}}
  }
 ]
}
