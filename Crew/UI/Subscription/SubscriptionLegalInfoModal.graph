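# Crew subscription legal modal that switches between subscription terms and EULA text, with player and subscriber-since panels alongside.
# PreConstruct feeds a DialogVM_EULA into the UIKit dialog; activation binds its button dispatcher and opens on Terms.
# Button indices drive confirm, EULA, Terms and back, swapping title, content and button visibility while logging analytics.
# Confirm broadcasts SubConfirmed then backs out, and SetPurchaseFlow shows that button only for purchases.
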WidgetBlueprint /Crew/UI/Subscription/SubscriptionLegalInfoModal
animations [
 ShowUp {
  binding_count=0
  length_seconds=0.14
  object=&self:ShowUp
  sequence {bindings [], display_rate {numerator=60}}
 }
 Dismiss {
  binding_count=0
  length_seconds=0.06
  object=&self:Dismiss
  sequence {bindings [], display_rate {numerator=100}}
 }
]
graphs [
 EventGraph {nodes [
   n1 = DialogButtonClicked!() -> (OutputDelegate, then, ButtonIndex:int) {is=title K2Node_CustomEvent}
   > n2 = GetPurchaseTraceId!(execute) -> (then, ReturnValue:Flow Analytics Additional Info Structure)
   > Set AnalyticsAdditionalInfo(AnalyticsAdditionalInfo:Flow Analytics Additional Info Structure=@n2) -> (Output_Get:Flow Analytics Additional Info Structure)
   > n3 = Cast To FortPlayerController(Object=GetOwningPlayer())
   > n4 = Switch on Int!(execute:Exec, Selection:int=@n1.ButtonIndex) -> (Default:Exec, 0:Exec, 1:Exec, 2:Exec, 3:Exec)
   ConfirmSub(execute=@n4.0)
   > "Fire Event Element Interacted (Localized)"!(
    execute
    FortPC:Fort Player Controller Object Reference=@n3
    ElementName:"Text (by ref)"="To Text (String)"(InString=Confirm Subscription)
    AdditionalInfo:Flow Analytics Additional Info Structure=Get AnalyticsAdditionalInfo() -> (AnalyticsAdditionalInfo:Flow Analytics Additional Info Structure)
   ) -> (then)
   EULAClicked(execute=@n4.1)
   > "Fire Event Element Interacted (Localized)"(FortPC=@n3, ElementName="To Text (String)"(InString=EULA Clicked), AdditionalInfo=Get AnalyticsAdditionalInfo())
   TermsClicked(execute=@n4.2)
   > "Fire Event Element Interacted (Localized)"(FortPC=@n3, ElementName="To Text (String)"(InString=Terms Clicked), AdditionalInfo=Get AnalyticsAdditionalInfo())
   TriggerModalBackout(execute=@n4.3, self:Subscription Modal Base Object Reference)
   > "Fire Event Element Interacted (Localized)"(FortPC=@n3, ElementName="To Text (String)"(InString=Back), AdditionalInfo=Get AnalyticsAdditionalInfo())
   ConfirmSub() {is=title K2Node_CustomEvent}
   > Call Sub Confirmed!(execute:Exec, self:Subscription Legal Info Modal Object Reference) -> (then:Exec) {is=title K2Node_CallDelegate}
   > TriggerModalBackout()
   EULAClicked() {is=title K2Node_CustomEvent}
   n5 = Get CrewVM() -> (CrewVM:Crew VM Object Reference)
   n6 = Get AnalyticsContext() -> (AnalyticsContext:String)
   > SendSubscriptionAnalyticsEvent!(execute, self:Crew VM Object Reference=@n5, ContextName:String=@n6, Interaction:String=EULAClicked) -> (then)
   > SetActiveWidget(
    self:Widget Switcher Object Reference=Get ContentSwitcher() -> (ContentSwitcher:Common Widget Switcher Legacy Object Reference)
    Widget=Get SubscriptionEulaWidget() -> (SubscriptionEulaWidget:Subscription Eula Widget Object Reference)
   )
   n7 = Get Dialog_Internal() -> (Dialog_Internal:WBP UIKit Dialog Internal Object Reference)
   > SetButtonVisibility!(execute, self:WBP UIKit Dialog Internal Object Reference=@n7, ButtonIndex:int=1, Visibility:ESlateVisibility Enum=Collapsed) -> (then)
   > SetButtonVisibility(self=@n7, ButtonIndex=2, Visibility=SelfHitTestInvisible)
   > SetVisibility(self=Get Text_AdditionalInfo() -> (Text_AdditionalInfo:Common Text Object Reference), InVisibility=SelfHitTestInvisible)
   n8 = Get Dialog_Internal()
   > SetDialogTitle!(execute, self:WBP UIKit Dialog Internal Object Reference=@n8, Title:Text="LOCTEXT(\"EULA\")") -> (then)
   TermsClicked() {is=title K2Node_CustomEvent}
   > n9 = SendSubscriptionAnalyticsEvent(self=@n5, ContextName=@n6, Interaction=TermsClicked)
   SetActiveWidget(
    execute=[@n12.then, @n9.then]
    self=Get ContentSwitcher()
    Widget=Get SubscriptionTermsWidget() -> (SubscriptionTermsWidget:Subscription Eula Widget Object Reference)
   )
   > SetButtonVisibility(self=@n7, ButtonIndex=1, Visibility=SelfHitTestInvisible)
   > SetButtonVisibility(self=@n7, ButtonIndex=2, Visibility=Collapsed)
   > SetVisibility(self=Get Text_AdditionalInfo(), InVisibility=Collapsed)
   > SetDialogTitle(self=@n8, Title="LOCTEXT(\"Subscription Terms\")")
   Event On Activated()
   > OnActivatedDefault(self:Subscription Modal Base Object Reference)
   n10 = Create Event()
   n11 = Get Dialog_Internal()
   > n12 = Bind Event to Dialog Button Clicked Dispatcher(self:WBP UIKit Dialog Internal Object Reference=@n11, Delegate=@n10) {is=title K2Node_AddDelegate}
   Event PreConstruct()
   > n13 = Construct Object from Class(Class=/Crew/UI/Subscription/Data/DialogVM_EULA.DialogVM_EULA_C) -> (ReturnValue:Dialog VM EULA Object Reference)
   > SetDialogData!(execute, self:WBP UIKit Dialog Internal Object Reference=Get Dialog_Internal(), Data:UI Kit Dialog Viewmodel Object Reference=@n13) -> (then)
   Event On Deactivated()
   > Unbind Event from Dialog Button Clicked Dispatcher(self:WBP UIKit Dialog Internal Object Reference=@n11, Delegate=@n10) {is=title K2Node_RemoveDelegate}
  ]}
 SetPurchaseFlow {nodes [
   n1 = SetPurchaseFlow!() -> (then, IsPurchase:bool=~) {is=title K2Node_FunctionEntry}
   > SetButtonVisibility(self=Get Dialog_Internal(), ButtonIndex=0, Visibility=Select(Option 0=Collapsed, Option 1=SelfHitTestInvisible, Index=@n1))
  ]}
 SubConfirmed {nodes [SubConfirmed() {is=title K2Node_FunctionEntry}]}
]
variables [
 AnalyticsAdditionalInfo {type=FlowAnalyticsAdditionalInfo, default {}}
 SubConfirmed {}
]
widgets {
 info {
  inheritedWidgetCount=0
  namedSlotCount=1
  parentClass=&/SubscriptionModalBase.SubscriptionModalBase_C
  rootWidgetClass=&Overlay
  widgetCount=12
 }
 widgets [
  Overlay_31 {widgetClassPath=&Overlay, widget=Overlay_31}
  Dialog_Internal {
   widgetClassPath=/FortUILibrary/Systems/Dialog/WBP_UIKit_Dialog_Internal
   bIsVariable=true
   parent=Overlay_31
   widget=Dialog_Internal
   properties {
    buttons {
     each Button Horizontal Fill=false
     each Button Min Width=157
     each Button Size=Medium
     focuse Button=0
     space Between Buttons=24
     cTA Buttons Types [Loud,Regular]
     group Placement {
      horizontal Alignment=HAlign_Fill
      vertical Alignment=VAlign_Bottom
      grid {column=1, column Span=0, layer=0, row=2, row Span=0, nudge=<0 0>}
      padding=<ltrb 80 64 80 48>
     }
    }
    dialogSIze {
     height Override {enable=true, value=850}
     max Aspect Ratio {enable=false, value=0}
     max Desired Height {enable=true, value=800}
     max Desired Width {enable=false, value=0}
     min Aspect Ratio {enable=false, value=0}
     min Desired Height {enable=false, value=0}
     min Desired Width {enable=false, value=0}
     width Override {enable=true, value=1490}
    }
    titleProperties {
     auto Wrap Text=true
     justification=Left
     line Height Percentage=1
     mobile Font Size Multiplier=1.3
     transform Policy=ToUpper
     wrap Text At=0
     color {colorUseRule=UseColor_Specified, specifiedColor=<rgba 1 1 1 1>}
     font {
      bForceMonospaced=false
      bMaterialIsStencil=false
      fontMaterial=null
      fontObject=&HeadingNow
      letterSpacing=0
      monospacedWidth=1
      size=25.5
      skewAmount=0
      typefaceFontName=86 Bold Italic
      outlineSettings {
       bApplyOutlineToDropShadows=false
       bMiteredCorners=false
       bSeparateFillAlpha=false
       outlineMaterial=null
       outlineSize=0
       outlineColor=<rgba 0 0 0 1>
      }
     }
     placement {
      horizontal Alignment=HAlign_Fill
      vertical Alignment=VAlign_Top
      grid {column=1, column Span=0, layer=0, row=0, row Span=0, nudge=<0 0>}
      padding=<ltrb 80 48 80 32>
     }
    }
   }
   slot {class=OverlaySlot, properties {horizontalAlignment=HAlign_Fill, verticalAlignment=VAlign_Fill}}
  }
  Overlay_0 {widgetClassPath=&Overlay, namedSlotHost=Dialog_Internal, widget=Overlay_0}
  HorizontalBox_1 {
   widgetClassPath=&HorizontalBox
   parent=Overlay_0
   widget=HorizontalBox_1
   slot {class=OverlaySlot, properties {horizontalAlignment=HAlign_Fill, verticalAlignment=VAlign_Fill}}
  }
  ContentVB {
   widgetClassPath=&VerticalBox
   parent=HorizontalBox_1
   widget=ContentVB
   slot {class=HorizontalBoxSlot, properties {padding=<ltrb 80 0 64 0>, size {sizeRule=Fill, value=1}}}
  }
  ContentSwitcher {
   widgetClassPath=&CommonWidgetSwitcherLegacy
   bIsVariable=true
   parent=ContentVB
   widget=ContentSwitcher
   slot {class=VerticalBoxSlot, properties {size {sizeRule=Fill, value=1}}}
  }
  SubscriptionTermsWidget {
   widgetClassPath=&/SubscriptionEulaWidget
   bIsVariable=true
   parent=ContentSwitcher
   widget=SubscriptionTermsWidget
   properties {eULAKeyToShow=fn-crew-terms}
   slot {class=WidgetSwitcherSlot}
  }
  SubscriptionEulaWidget {
   widgetClassPath=&SubscriptionEulaWidget
   bIsVariable=true
   parent=ContentSwitcher
   widget=SubscriptionEulaWidget
   slot {class=WidgetSwitcherSlot}
  }
  Text_AdditionalInfo {
   widgetClassPath=&CommonTextBlock
   bIsVariable=true
   parent=ContentVB
   widget=Text_AdditionalInfo
   properties {
    autoWrapText=true
    text="Additional information can be found in the \"Legal\" section of the  main menu"
    textTransformPolicy=ToUpper
    colorAndOpacity {colorUseRule=UseColor_Specified, specifiedColor=<rgba 0.5776 0.5776 0.5776 1>}
    font {
     bForceMonospaced=false
     bMaterialIsStencil=false
     fontMaterial=null
     fontObject=&HeadingNow
     letterSpacing=0
     monospacedWidth=1
     size=14
     skewAmount=0
     typefaceFontName=86 Bold
     outlineSettings {
      bApplyOutlineToDropShadows=false
      bMiteredCorners=false
      bSeparateFillAlpha=false
      outlineMaterial=null
      outlineSize=0
      outlineColor=<rgba 1 1 1 0>
     }
    }
    shadowOffset=<0 0>
   }
   slot {class=VerticalBoxSlot, properties {horizontalAlignment=HAlign_Left, verticalAlignment=VAlign_Bottom, padding=<ltrb 0 16 0 0>}}
  }
  ImageContent {
   widgetClassPath=&VerticalBox
   parent=HorizontalBox_1
   widget=ImageContent
   slot {class=HorizontalBoxSlot, properties {padding=<ltrb 0 0 54 0>}}
  }
  WBP_SubscriberSince {
   widgetClassPath=&/WBP_SubscriberSince
   bIsVariable=true
   parent=ImageContent
   widget=WBP_SubscriberSince
   slot {class=VerticalBoxSlot, properties {verticalAlignment=VAlign_Center, size {sizeRule=Fill, value=1}}}
  }
  WBP_PlayerInfo {
   widgetClassPath=/Game/UI/Frontend/ItemShop/MultiProduct/WBP_PlayerInfo
   bIsVariable=true
   parent=Overlay_31
   widget=WBP_PlayerInfo
   slot {class=OverlaySlot, properties {horizontalAlignment=HAlign_Right, padding=<ltrb 0 40 80 0>}}
  }
 ]
}
